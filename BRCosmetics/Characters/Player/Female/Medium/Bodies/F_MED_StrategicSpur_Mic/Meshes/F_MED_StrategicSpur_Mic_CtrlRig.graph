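# Control Rig for the F_MED_StrategicSpur_Mic cosmetic body, adding corrective bone offsets for shoulder pads, belt, lapels and armpits.
# It runs the shared DeformRig function, then uses spherical pose readers and root animation attributes to weight ModifyTransforms nodes.
# A raised right clavicle during parachuting or custom emotes sets isHangGliding, replacing the shoulder-pad function with dedicated offsets.
# Belt, lapel and armpit functions run on both branches.
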ControlRigBlueprint /BRCosmetics/Characters/Player/Female/Medium/Bodies/F_MED_StrategicSpur_Mic/Meshes/F_MED_StrategicSpur_Mic_CtrlRig
graphs [
 New Function_ContainedGraph {nodes [
   n1 = RigVMFunctionEntryNode()
   RigVMFunction_Sequence = &RigVMFunction_Sequence(ExecuteContext=@n1.ExecuteContext)
   RigVMFunction_Sequence_1 = RigVMAggregateNode(ExecuteContext=@RigVMFunction_Sequence.A)
   RigVMFunction_ControlFlowBranch = &RigVMFunction_ControlFlowBranch(ExecuteContext=@RigVMFunction_Sequence_1.A, Condition=RigVMVariableNode() {variable=PlayingEmote}.Value)
   n2 = &RigUnit_SphericalPoseReader(
    ActiveRegionScaleFactors=(PositiveWidth=1,NegativeWidth=1,PositiveHeight=1,NegativeHeight=1)
    ActiveRegionSize=0.05
    Debug=(bDrawDebug=false,bDraw2D=false,bDrawLocalAxes=False,DebugScale=25,DebugSegments=20,DebugThickness=0.25)
    DriverAxis=(X=1,Y=0,Z=0)
    DriverItem=(Type=Bone,Name="clavicle_l")
    FalloffRegionScaleFactors=(PositiveWidth=1,NegativeWidth=1,PositiveHeight=1,NegativeHeight=1)
    FalloffSize=0.3
    FlipHeightScaling=False
    FlipWidthScaling=False
    OptionalParentItem=(Type=Bone,Name="None")
    RotationOffset=(X=0,Y=-25,Z=0)
    ExecutePin=@RigVMFunction_ControlFlowBranch.False
   )
   > &RigUnit_ModifyTransforms(
    ItemToModify=((Item=(Type=Bone,Name="base_sholderpad_l"),Transform=(Rotation=(X=0.0871,Y=-0.0183,Z=0.0016,W=0.996),Translation=(X=-1.5,Y=1.4018,Z=-1.2706),Scale3D=(X=1,Y=1,Z=1))))
    WeightMinimum=0
    Weight=@n2.OutputParam
   )
   > Get Mirror Transform_5 = RigVMFunctionReferenceNode(transform=(Rotation=(X=-0.0852,Y=-0.0259,Z=0.0852,W=0.9924),Translation=(X=1,Y=-1.4,Z=-1.27),Scale3D=(X=1,Y=1,Z=1)))
   n3 = RigVMVariableNode() {variable=ClavUp_R}
   > &RigUnit_ModifyTransforms(
    ItemToModify=((Item=(Type=Bone,Name="base_sholderpad_r"),Transform=(Rotation=(X=0.0106,Y=-0.1216,Z=0.0865,W=0.9887),Translation=(X=-2.9701,Y=0,Z=-2.6751),Scale3D=(X=1,Y=1,Z=1))))
    WeightMinimum=0
    ItemToModify.0.Transform=@"Get Mirror Transform_5".mirrorTransform
    Weight=@n3.Value
   )
   > RigVMVariableNode(Value=@n2.OutputParam) {variable=L_clav_raise_spr}
   > RigVMVariableNode(Value=@n3.Value) {variable=r_Clav_raise_spr}
   n4 = &RigUnit_SphericalPoseReader(
    ActiveRegionScaleFactors=(PositiveWidth=2,NegativeWidth=1,PositiveHeight=2,NegativeHeight=2)
    ActiveRegionSize=0.1
    Debug=(bDrawDebug=false,bDraw2D=false,DebugScale=25,DebugSegments=20,DebugThickness=0.25)
    DriverAxis=(X=1,Y=0,Z=0)
    DriverItem=(Type=Bone,Name="upperarm_l")
    FalloffRegionScaleFactors=(PositiveWidth=1,NegativeWidth=0.5,PositiveHeight=1,NegativeHeight=1)
    FalloffSize=0.5
    OptionalParentItem=(Type=Bone,Name="clavicle_l")
    RotationOffset=(X=0,Y=20,Z=0)
    ExecutePin=@RigVMFunction_Sequence_1.B
   )
   > n5 = &RigUnit_SphericalPoseReader(
    ActiveRegionScaleFactors=(PositiveWidth=2,NegativeWidth=1,PositiveHeight=2,NegativeHeight=2)
    ActiveRegionSize=0.1
    Debug=(bDrawDebug=false,bDraw2D=false,DebugScale=25,DebugSegments=20,DebugThickness=0.25)
    DriverAxis=(X=-1,Y=0,Z=0)
    DriverItem=(Type=Bone,Name="upperarm_r")
    FalloffRegionScaleFactors=(PositiveWidth=1,NegativeWidth=0.5,PositiveHeight=1,NegativeHeight=1)
    FalloffSize=0.5
    FlipHeightScaling=true
    OptionalParentItem=(Type=Bone,Name="clavicle_r")
    RotationOffset=(X=180,Y=-20,Z=180)
   )
   > &RigUnit_ModifyTransforms(
    ItemToModify="((Item=(Type=Bone,Name=\"base_sholderpad_l\"),Transform=(Rotation=(X=0.0858,Y=-0.173,Z=0.0151,W=0.9811),Translation=(X=0.5,Y=-1,Z=1),Scale3D=(X=0.9,Y=0.99,Z=0.9))),(Item=(Type=Bone,Name=\"None\"),Transform=(Rotation=(X=0,Y=-0.0499,Z=0,W=0.9988),Translation=(X= ...(truncated)"
    WeightMinimum=0
    Weight=@n4.OutputParam
   )
   > Get Mirror Transform = RigVMFunctionReferenceNode(transform=(Rotation=(X=0,Y=-0.2588,Z=0,W=0.9659),Translation=(X=0.4,Y=-1,Z=1),Scale3D=(X=0.9,Y=0.99,Z=0.9)))
   > Get Mirror Transform_1 = RigVMFunctionReferenceNode(transform=(Rotation=(X=0,Y=-0.0499,Z=0,W=0.9988),Translation=(X=-1.9805,Y=-0.0338,Z=-0.5915),Scale3D=(X=1,Y=1,Z=1)))
   > Get Mirror Transform_2 = RigVMFunctionReferenceNode(transform=(Rotation=(X=0.0858,Y=-0.173,Z=0.0151,W=0.9811),Translation=(X=0.5,Y=-1,Z=1),Scale3D=(X=0.9,Y=0.99,Z=0.9)))
   > &RigUnit_ModifyTransforms(
    ItemToModify="((Item=(Type=Bone,Name=\"base_sholderpad_r\"),Transform=(Rotation=(X=0,Y=-0.1962,Z=0,W=0.9806),Translation=(X=0.4714,Y=0,Z=2.3308),Scale3D=(X=0.9,Y=0.9,Z=0.9))),(Item=(Type=Bone,Name=\"base_sholderpad_b_r\"),Transform=(Rotation=(X=0,Y=-0.0499,Z=0,W=0.9988),Tra ...(truncated)"
    WeightMinimum=0
    ItemToModify.0.Transform=@"Get Mirror Transform".mirrorTransform
    ItemToModify.1.Transform=@"Get Mirror Transform_1".mirrorTransform
    ItemToModify.2.Transform=@"Get Mirror Transform_2".mirrorTransform
    Weight=@n5.OutputParam
   )
   > &RigUnit_ModifyTransforms(
    ItemToModify=((Item=(Type=Bone,Name="deform_upperarm_bk_r"),Transform=(Rotation=(X=0,Y=0,Z=0,W=1),Translation=(X=-15,Y=10,Z=15),Scale3D=(X=1,Y=1,Z=1))))
    WeightMinimum=0
    Weight=&RigVMFunction_MathDoubleMul(A=@n5.OutputParam, B=RigVMVariableNode() {variable=ArmpitFix}.Value).Result
   )
   Add = &RigVMFunction_MathDoubleAdd(A=@n4.OutputParam, B=RigVMVariableNode() {variable=L_clav_raise_spr}.Value)
   Remap = &RigVMFunction_MathDoubleRemap(SourceMaximum=1.5, SourceMinimum=1, TargetMaximum=1, TargetMinimum=0, Value=@Add.Result)
   n6 = &RigVMFunction_AlphaInterp(InterpSpeedDecreasing=5, InterpSpeedIncreasing=5, Scale=1, bClampResult=true, bInterpResult=False, bMapRange=true, Value=@Remap.Result)
   > &RigUnit_ModifyTransforms(
    ItemToModify="((Item=(Type=Bone,Name=\"base_sholderpad_l\"),Transform=(Rotation=(X=0,Y=0,Z=0.4226,W=0.9063),Translation=(X=3,Y=-5.3,Z=0),Scale3D=(X=1,Y=1,Z=1))),(Item=(Type=Bone,Name=\"None\"),Transform=(Rotation=(X=0,Y=0,Z=0,W=1),Translation=(X=0,Y=0,Z=0),Scale3D=(X=1,Y=1, ...(truncated)"
    WeightMinimum=0
    Weight=@n6.Result
   )
   > Get Mirror Transform_1_1 = RigVMFunctionReferenceNode(transform=(Rotation=(X=0,Y=0,Z=0.4226,W=0.9063),Translation=(X=3,Y=-2.5,Z=0),Scale3D=(X=1,Y=1,Z=1)))
   n7 = RigVMVariableNode() {variable=r_Clav_raise_spr}
   > &RigUnit_ModifyTransforms(
    ItemToModify="((Item=(Type=Bone,Name=\"base_sholderpad_r\"),Transform=(Rotation=(X=0,Y=-0.1962,Z=0,W=0.9806),Translation=(X=0.4714,Y=0,Z=2.3308),Scale3D=(X=0.9,Y=0.9,Z=0.9))),(Item=(Type=Bone,Name=\"base_sholderpad_f_r\"),Transform=(Rotation=(X=0,Y=0,Z=0,W=1),Translation=(X ...(truncated)"
    WeightMinimum=0
    ItemToModify.0.Transform=@"Get Mirror Transform_1_1".mirrorTransform
    Weight=&RigVMFunction_MathDoubleRemap(
     SourceMaximum=1.6
     SourceMinimum=1
     TargetMaximum=1
     TargetMinimum=0
     Value=&RigVMFunction_MathDoubleAdd(A=@n5.OutputParam, B=@n7.Value).Result
    ).Result
   )
   n8 = &RigUnit_SphericalPoseReader(
    ActiveRegionScaleFactors=(PositiveWidth=2,NegativeWidth=1,PositiveHeight=1,NegativeHeight=2)
    ActiveRegionSize=0.12
    Debug=(bDrawDebug=false,DebugScale=25,DebugSegments=20,DebugThickness=0.25)
    DriverAxis=(X=1,Y=0,Z=0)
    DriverItem=(Type=Bone,Name="upperarm_l")
    FalloffRegionScaleFactors=(PositiveWidth=0.9,NegativeWidth=1,PositiveHeight=1,NegativeHeight=1)
    FalloffSize=0.6
    OptionalParentItem=(Type=Bone,Name="clavicle_l")
    RotationOffset=(X=50,Y=0,Z=10)
    ExecutePin=@RigVMFunction_Sequence_1.C
   )
   > n9 = &RigUnit_SphericalPoseReader(
    ActiveRegionScaleFactors=(PositiveWidth=1,NegativeWidth=2,PositiveHeight=1,NegativeHeight=2)
    ActiveRegionSize=0.12
    Debug=(bDrawDebug=false,DebugScale=25,DebugSegments=20,DebugThickness=0.25)
    DriverAxis=(X=-1,Y=0,Z=0)
    DriverItem=(Type=Bone,Name="upperarm_r")
    FalloffRegionScaleFactors=(PositiveWidth=1,NegativeWidth=1,PositiveHeight=1,NegativeHeight=1)
    FalloffSize=0.48
    FlipHeightScaling=true
    FlipWidthScaling=true
    OptionalParentItem=(Type=Bone,Name="clavicle_r")
    RotationOffset=(X=121.4949,Y=-8.2258,Z=-147.8671)
   )
   > &RigUnit_ModifyTransforms(
    ItemToModify="((Item=(Type=Bone,Name=\"base_sholderpad_l\"),Transform=(Rotation=(X=0,Y=0,Z=0.2588,W=0.9659),Translation=(X=0,Y=-3,Z=-0.5),Scale3D=(X=1,Y=1.15,Z=1))),(Item=(Type=Bone,Name=\"None\"),Transform=(Rotation=(X=0.038,Y=-0.0285,Z=0.0011,W=0.9989),Translation=(X=-2.2 ...(truncated)"
    WeightMinimum=0
    Weight=@n8.OutputParam
   )
   > Get Mirror Transform_3 = RigVMFunctionReferenceNode(transform=(Rotation=(X=0,Y=0,Z=0.342,W=0.9397),Translation=(X=0,Y=-3,Z=-0.5),Scale3D=(X=1,Y=1.15,Z=1)))
   > &RigUnit_ModifyTransforms(
    ItemToModify="((Item=(Type=Bone,Name=\"base_sholderpad_r\"),Transform=(Rotation=(X=0,Y=-0.1632,Z=0,W=0.9866),Translation=(X=-2.3381,Y=4.0995,Z=-3.9185),Scale3D=(X=1,Y=1,Z=1))),(Item=(Type=Bone,Name=\"None\"),Transform=(Rotation=(X=0.038,Y=-0.0285,Z=0.0011,W=0.9989),Translat ...(truncated)"
    WeightMinimum=0
    ItemToModify.0.Transform=@"Get Mirror Transform_3".mirrorTransform
    Weight=@n9.OutputParam
   )
   RigVMFunction_MathBoolOr = &RigVMFunction_MathBoolOr(A=RigVMVariableNode() {variable=Parachuting}.Value, B=RigVMVariableNode() {variable=PlayingEmote}.Value)
   RigVMFunction_MathBoolToFloat = &RigVMFunction_MathBoolToFloat(Value=@RigVMFunction_MathBoolOr.Result)
   Multiply = &RigVMFunction_MathDoubleMul(
    B=RigVMDispatchNode(BoneName=root, Default=0, Name=deformRig_l_upperarm_up_alpha) {template="DISPATCH_RigDispatch_GetAnimAttribute(in Name,in BoneName,in Default,out Value,out Found)"}.Value
    A=@RigVMFunction_MathBoolToFloat.Result
   )
   &RigUnit_ModifyTransforms(
    ItemToModify=((Item=(Type=Bone,Name="base_sholderpad_l"),Transform=(Rotation=(X=0,Y=0,Z=0.1736,W=0.9848),Translation=(X=2,Y=-1,Z=1),Scale3D=(X=0.9,Y=0.9,Z=0.9))))
    WeightMinimum=0
    Weight=@Multiply.Result
    ExecutePin=@RigVMFunction_Sequence_1.D
   )
   > &RigUnit_ModifyTransforms(
    ItemToModify=((Item=(Type=Bone,Name="base_sholderpad_r"),Transform=(Rotation=(X=-0.0019,Y=0.0436,Z=0.0436,W=0.9981),Translation=(X=1,Y=1.5,Z=0),Scale3D=(X=0.9,Y=1,Z=0.9))))
    WeightMinimum=0
    Weight=&RigVMFunction_MathDoubleMul(
     B=RigVMDispatchNode(BoneName=root, Default=0, Name=deformRig_r_upperarm_up_alpha) {template="DISPATCH_RigDispatch_GetAnimAttribute(in Name,in BoneName,in Default,out Value,out Found)"}.Value
     A=@RigVMFunction_MathBoolToFloat.Result
    ).Result
   )
   RigVMFunctionReturnNode(ExecuteContext=@RigVMFunction_Sequence.B)
   &RigUnit_SphericalPoseReader(
    ActiveRegionScaleFactors=(PositiveWidth=1,NegativeWidth=1,PositiveHeight=1,NegativeHeight=1)
    ActiveRegionSize=0.05
    Debug=(bDrawDebug=false,bDraw2D=false,bDrawLocalAxes=False,DebugScale=25,DebugSegments=20,DebugThickness=0.25)
    DriverAxis=(X=-1,Y=0,Z=0)
    DriverItem=(Type=Bone,Name="clavicle_r")
    ExecutePin=()
    FalloffRegionScaleFactors=(PositiveWidth=1,NegativeWidth=1,PositiveHeight=1,NegativeHeight=1)
    FalloffSize=0.3
    FlipHeightScaling=False
    FlipWidthScaling=False
    OptionalParentItem=(Type=Bone,Name="None")
    RotationOffset=(X=180,Y=-25,Z=0)
   )
   RigVMVariableNode() {variable=parachute open or emote}
  ]}
 New Function_ContainedGraph {nodes [
   RigVMFunctionEntryNode()
   > RigVMFunction_Sequence_1_1_1_1 = RigVMAggregateNode()
   n1 = &RigUnit_SphericalPoseReader(
    ActiveRegionScaleFactors=(PositiveWidth=1.5,NegativeWidth=1,PositiveHeight=4.5,NegativeHeight=3)
    ActiveRegionSize=0.1
    Debug=(bDrawDebug=false,DebugScale=25,DebugSegments=20,DebugThickness=0.25)
    DriverAxis=(X=-1,Y=0,Z=0)
    DriverItem=(Type=Bone,Name="thigh_r")
    FalloffRegionScaleFactors=(PositiveWidth=1,NegativeWidth=1.5,PositiveHeight=5,NegativeHeight=1)
    FalloffSize=0.45
    OptionalParentItem=(Type=Bone)
    RotationOffset=(X=90,Y=180,Z=-15)
    ExecutePin=@RigVMFunction_Sequence_1_1_1_1.B
   )
   n2 = RigVMVariableNode() {variable=BeltFix}
   > &RigUnit_ModifyTransforms(
    ItemToModify=((Item=(Type=Bone,Name="belt_helper_r"),Transform=(Rotation=(X=0,Y=0,Z=0,W=1),Translation=(X=-3.5,Y=0,Z=0),Scale3D=(X=1,Y=1,Z=1))))
    Weight=&RigVMFunction_MathDoubleRemap(
     SourceMaximum=1
     SourceMinimum=0
     TargetMaximum=1
     TargetMinimum=0
     bClamp=False
     Value=&RigVMFunction_MathDoubleMul(A=@n1.OutputParam, B=@n2.Value).Result
    ).Result
    ItemToModify.0.Transform.Rotation=&RigVMFunction_MathQuaternionFromEuler(Euler=(X=-5,Y=5,Z=4)).Result
   )
   RigVMFunctionReturnNode(ExecuteContext=@RigVMFunction_Sequence_1_1_1_1.C)
   Remap = &RigVMFunction_MathDoubleRemap(
    SourceMaximum=1
    SourceMinimum=0
    TargetMaximum=1
    TargetMinimum=0
    bClamp=False
    Value=RigVMDispatchNode(BoneName=root, Default=0, Name=deformRig_l_thigh_up_alpha) {template="DISPATCH_RigDispatch_GetAnimAttribute(in Name,in BoneName,in Default,out Value,out Found)"}.Value
   )
   &RigUnit_ModifyTransforms(
    ItemToModify="((Item=(Type=Bone,Name=\"belt_helper_l\"),Transform=(Rotation=(X=0,Y=0,Z=0,W=1),Translation=(X=-9.5,Y=-5,Z=-1),Scale3D=(X=1,Y=1,Z=1))),(Item=(Type=Bone,Name=\"base_cord\"),Transform=(Rotation=(X=0,Y=0,Z=0,W=1),Translation=(X=-9,Y=0.5,Z=0.7),Scale3D=(X=1,Y=1,Z= ...(truncated)"
    Weight=@Remap.Result
    ItemToModify.0.Transform.Rotation=&RigVMFunction_MathQuaternionFromEuler(Euler=(X=0,Y=-10,Z=20)).Result
    ItemToModify.1.Transform.Rotation=&RigVMFunction_MathQuaternionFromEuler(Euler=(X=0,Y=20,Z=10)).Result
    ExecutePin=@RigVMFunction_Sequence_1_1_1_1.A
   )
  ]}
 New Function_ContainedGraph {nodes [
   RigVMFunctionEntryNode()
   > RigVMFunction_Sequence = RigVMAggregateNode()
   n1 = &RigUnit_SphericalPoseReader(
    ActiveRegionScaleFactors=(PositiveWidth=3,NegativeWidth=2.3,PositiveHeight=4,NegativeHeight=0.5)
    ActiveRegionSize=0.07
    Debug=(bDrawDebug=false,DebugScale=25,DebugSegments=20,DebugThickness=0.25)
    DriverAxis=(X=1,Y=0,Z=0)
    DriverItem=(Type=Bone,Name="upperarm_l")
    FalloffRegionScaleFactors=(PositiveWidth=1,NegativeWidth=1,PositiveHeight=1.5,NegativeHeight=0.5)
    FalloffSize=0.4
    FlipWidthScaling=true
    OptionalParentItem=(Type=Bone)
    RotationOffset=(X=-92,Y=200,Z=190)
    ExecutePin=@RigVMFunction_Sequence.A
   )
   > &RigUnit_ModifyTransforms(
    ItemToModify=((Item=(Type=Bone,Name="base_lower_lapel_l"),Transform=(Rotation=(X=0,Y=0,Z=0,W=1),Translation=(X=0,Y=0,Z=0),Scale3D=(X=1,Y=1,Z=1))))
    ItemToModify.0.Transform.Rotation=&RigVMFunction_MathQuaternionFromEuler(Euler=(X=0,Y=-8,Z=95)).Result
    Weight=@n1.OutputParam
   )
   > &RigUnit_ModifyTransforms(
    ItemToModify=((Item=(Type=Bone,Name="base_upper_lapel_l"),Transform=(Rotation=(X=0,Y=0,Z=0.7071,W=0.7071),Translation=(X=0,Y=0,Z=0),Scale3D=(X=1,Y=1,Z=1))))
    Weight=@n1.OutputParam
   )
   n2 = &RigUnit_SphericalPoseReader(
    ActiveRegionScaleFactors=(PositiveWidth=3.5,NegativeWidth=2,PositiveHeight=4,NegativeHeight=1.5)
    ActiveRegionSize=0.07
    Debug=(bDrawDebug=False,DebugScale=25,DebugSegments=20,DebugThickness=0.25)
    DriverAxis=(X=-1,Y=0,Z=0)
    DriverItem=(Type=Bone,Name="upperarm_r")
    FalloffRegionScaleFactors=(PositiveWidth=0.5,NegativeWidth=0.2,PositiveHeight=1.5,NegativeHeight=0.8)
    FalloffSize=0.4
    FlipHeightScaling=false
    FlipWidthScaling=true
    OptionalParentItem=(Type=Bone)
    RotationOffset=(X=90,Y=230,Z=190)
    ExecutePin=@RigVMFunction_Sequence.B
   )
   > &RigUnit_ModifyTransforms(
    ItemToModify=((Item=(Type=Bone,Name="base_lower_lapel_r"),Transform=(Rotation=(X=0,Y=0,Z=0,W=1),Translation=(X=0,Y=1,Z=1),Scale3D=(X=1,Y=1,Z=1))))
    ItemToModify.0.Transform.Rotation=&RigVMFunction_MathQuaternionFromEuler(Euler=(X=0,Y=-50,Z=130)).Result
    Weight=@n2.OutputParam
   )
   > &RigUnit_ModifyTransforms(
    ItemToModify=((Item=(Type=Bone,Name="base_upper_lapel_r"),Transform=(Rotation=(X=0,Y=0,Z=0.7071,W=0.7071),Translation=(X=0,Y=0,Z=0),Scale3D=(X=1,Y=1,Z=1))))
    ItemToModify.0.Transform.Rotation=&RigVMFunction_MathQuaternionFromEuler(Euler=(X=0,Y=0,Z=90)).Result
    Weight=@n2.OutputParam
   )
   RigVMFunctionReturnNode(ExecuteContext=@RigVMFunction_Sequence.C)
   RigVMFunction_Sequence_1 = RigVMAggregateNode(ExecuteContext=())
   RigUnit_SetTransform = &RigUnit_SetTransform(
    Item=(Type=Bone,Name="upperarm_l")
    Weight=1
    bInitial=False
    Value=&RigUnit_GetTransform(Item=(Type=Control,Name="upperarm_l_ctrl"), bInitial=False).Transform
    ExecutePin=@RigVMFunction_Sequence_1.A
   )
   RigUnit_GetTransform = &RigUnit_GetTransform(Item=(Type=Control,Name="upperarm_r_ctrl"), bInitial=False)
   &RigUnit_SetTransform(Item=(Type=Bone,Name="upperarm_r"), Weight=1, bInitial=False, Value=@RigUnit_GetTransform.Transform, ExecutePin=@RigVMFunction_Sequence_1.B)
  ]}
 New Function_ContainedGraph {nodes [
   RigVMFunctionEntryNode()
   > RigVMFunction_Sequence = RigVMAggregateNode()
   n1 = &RigUnit_SphericalPoseReader(
    ActiveRegionScaleFactors=(PositiveWidth=1,NegativeWidth=1,PositiveHeight=1.25,NegativeHeight=1)
    ActiveRegionSize=0.15
    Debug=(bDrawDebug=false,DebugScale=25,DebugSegments=20,DebugThickness=0.25)
    DriverAxis=(X=1,Y=0,Z=0)
    DriverItem=(Type=Bone,Name="upperarm_l")
    FalloffRegionScaleFactors=(PositiveWidth=1,NegativeWidth=1,PositiveHeight=1.25,NegativeHeight=1)
    FalloffSize=0.3
    OptionalParentItem=(Type=Bone,Name="clavicle_l")
    RotationOffset=(X=-5,Y=-110,Z=0)
    ExecutePin=@RigVMFunction_Sequence.A
   )
   Multiply = &RigVMFunction_MathDoubleMul(A=@n1.OutputParam, B=RigVMVariableNode() {variable=ShoulderFix}.Value)
   > &RigUnit_ModifyTransforms(
    ItemToModify="((Item=(Type=Bone,Name=\"base_shirtmover_l\"),Transform=(Rotation=(X=0,Y=0,Z=0,W=1),Translation=(X=-5.5,Y=-5,Z=0),Scale3D=(X=1,Y=1,Z=1))),(Item=(Type=Bone,Name=\"base_upperarm_inner_l\"),Transform=(Rotation=(X=0,Y=0,Z=0,W=1),Translation=(X=0,Y=2,Z=0),Scale3D=( ...(truncated)"
    WeightMinimum=0
    Weight=@Multiply.Result
   )
   > n2 = &RigUnit_SphericalPoseReader(
    ActiveRegionScaleFactors=(PositiveWidth=1,NegativeWidth=1,PositiveHeight=1.25,NegativeHeight=1)
    ActiveRegionSize=0.15
    Debug=(bDrawDebug=false,DebugScale=25,DebugSegments=20,DebugThickness=0.25)
    DriverAxis=(X=-1,Y=0,Z=0)
    DriverItem=(Type=Bone,Name="upperarm_r")
    FalloffRegionScaleFactors=(PositiveWidth=1,NegativeWidth=1,PositiveHeight=1.25,NegativeHeight=1)
    FalloffSize=0.3
    FlipHeightScaling=true
    FlipWidthScaling=true
    OptionalParentItem=(Type=Bone,Name="clavicle_r")
    RotationOffset=(X=5,Y=65,Z=0)
   )
   > &RigUnit_ModifyTransforms(
    ItemToModify="((Item=(Type=Bone,Name=\"base_shirtmover_r\"),Transform=(Rotation=(X=0,Y=0,Z=0,W=1),Translation=(X=5.5,Y=5,Z=0),Scale3D=(X=1,Y=1,Z=1))),(Item=(Type=Bone,Name=\"base_upperarm_inner_r\"),Transform=(Rotation=(X=0,Y=0,Z=0,W=1),Translation=(X=0,Y=0,Z=-2),Scale3D=(X ...(truncated)"
    WeightMinimum=0
    Weight=&RigVMFunction_MathDoubleMul(A=@n2.OutputParam, B=RigVMVariableNode() {variable=ShoulderFix}.Value).Result
   )
   RigVMFunctionReturnNode(ExecuteContext=@RigVMFunction_Sequence.B)
  ]}
 RigVMFunctionLibrary {nodes [
   Deform_Armpit = RigVMCollapseNode(ExecuteContext=())
   Deform_Belt = RigVMCollapseNode(ExecuteContext=())
   Deform_Lapel = RigVMCollapseNode(ExecuteContext=())
   Deform_SholderPads = RigVMCollapseNode(ExecuteContext=())
  ]}
 RigVMFunction_Sequence_1_1_1_1_1_ContainedGraph {nodes [
   n1 = RigVMFunctionEntryNode()
   n2 = &RigVMFunction_Sequence(ExecuteContext=@n1.ExecuteContext)
   n3 = &RigVMFunction_Sequence(ExecuteContext=@n2.B)
   RigVMFunctionReturnNode(A=@n2.A, B=@n3.A, C=@n3.B)
  ]}
 RigVMFunction_Sequence_1_1_ContainedGraph {nodes [
   n1 = RigVMFunctionEntryNode()
   n2 = &RigVMFunction_Sequence(ExecuteContext=@n1.ExecuteContext)
   n3 = &RigVMFunction_Sequence(ExecuteContext=@n2.B)
   n4 = &RigVMFunction_Sequence(ExecuteContext=@n3.B)
   n5 = &RigVMFunction_Sequence(ExecuteContext=@n4.B)
   RigVMFunctionReturnNode(A=@n2.A, B=@n3.A, C=@n4.A, D=@n5.A, E=@n5.B)
  ]}
 RigVMFunction_Sequence_1_ContainedGraph {nodes [
   n1 = RigVMFunctionEntryNode()
   RigVMFunction_Sequence = &RigVMFunction_Sequence(ExecuteContext=@n1.ExecuteContext)
   n2 = &RigVMFunction_Sequence(ExecuteContext=@RigVMFunction_Sequence.B)
   RigVMFunctionReturnNode(A=@RigVMFunction_Sequence.A, B=@n2.A, C=@n2.B)
  ]}
 RigVMFunction_Sequence_1_ContainedGraph {nodes [
   n1 = RigVMFunctionEntryNode()
   RigVMFunction_Sequence = &RigVMFunction_Sequence(ExecuteContext=@n1.ExecuteContext)
   n2 = &RigVMFunction_Sequence(ExecuteContext=@RigVMFunction_Sequence.B)
   RigVMFunctionReturnNode(A=@RigVMFunction_Sequence.A, B=@n2.A, C=@n2.B)
  ]}
 RigVMFunction_Sequence_1_ContainedGraph {nodes [
   n1 = RigVMFunctionEntryNode()
   RigVMFunction_Sequence = &RigVMFunction_Sequence(ExecuteContext=@n1.ExecuteContext)
   n2 = &RigVMFunction_Sequence(ExecuteContext=@RigVMFunction_Sequence.B)
   RigVMFunctionReturnNode(A=@RigVMFunction_Sequence.A, B=@n2.A, C=@n2.B)
  ]}
 RigVMModel {nodes [
   RigUnit_BeginExecution = &RigUnit_BeginExecution()
   FN_DeformRig_FNC = RigVMFunctionReferenceNode(
    DeformRig_ Curve="(deformRig_l_upperarm_up_alpha_72_#1=False,deformRig_r_upperarm_up_alpha_73_#2=true,deformRig_l_upperarm_fr_alpha_74_#3=False,deformRig_r_upperarm_fr_alpha_75_CB51B80 ...(truncated)"
    DeformRig_Bool="(upperarm_1_#4=True,elbow_3_#5=True,wrist_7_#6=True,groin_8_#7=true,glute_17_#8=true,knee_9_40C408324CA261 ...(truncated)"
    DeformRig_Debug=(upperarm_debug_18_#4=false,wrist_debug_20_#6=False,lat_debug_26_#9=false,groinGlute_debug_27_#7=False)
    DeformRig_Value="(upperarm_up_18_#10=(Rotation=(X=0.0,Y=-0.0,Z=0.0,W=1.0),Translation=(X=0.0,Y=0.0,Z=4),Scale3D=(X=0.70,Y=1.0,Z=1.0)),upperarm_fwd_fr_19_#11=(Rotation=(X=0,Y=0,Z=0,W=1),Translation=(X=-10,Y=10,Z=0),S ...(truncated)"
    ExecuteContext=@RigUnit_BeginExecution.ExecutePin
   )
   RigVMFunction_Sequence = &RigVMFunction_Sequence(ExecuteContext=@FN_DeformRig_FNC.ExecuteContext)
   RigVMFunction_ControlFlowBranch = &RigVMFunction_ControlFlowBranch(Condition=RigVMVariableNode() {variable=isHangGliding}.Value, ExecuteContext=@RigVMFunction_Sequence.B)
   Deform_SholderPads = RigVMFunctionReferenceNode(ExecuteContext=@RigVMFunction_ControlFlowBranch.False)
   > Deform_Belt = RigVMFunctionReferenceNode()
   > Deform_Lapel = RigVMFunctionReferenceNode()
   > Deform_Armpit = RigVMFunctionReferenceNode()
   &RigUnit_ModifyTransforms(
    ItemToModify="((Item=(Type=Bone,Name=\"base_sholderpad_r\"),Transform=(Rotation=(X=0,Y=-0.0501,Z=0,W=0.9987),Translation=(X=0,Y=0,Z=-1.6489),Scale3D=(X=1,Y=1,Z=1))),(Item=(Type=Bone,Name=\"base_sholderpad_l\"),Transform=(Rotation=(X=0,Y=-0.0463,Z=0,W=0.9989),Translation=(X= ...(truncated)"
    WeightMinimum=0
    Weight=RigVMVariableNode() {variable=ClavUp_Gliding}.Value
    ExecutePin=@RigVMFunction_ControlFlowBranch.True
   )
   > &RigUnit_ModifyTransforms(
    ItemToModify="((Item=(Type=Bone,Name=\"base_sholderpad_r\"),Transform=(Rotation=(X=0,Y=-0.0541,Z=0,W=0.9985),Translation=(X=0,Y=0,Z=0),Scale3D=(X=1,Y=1,Z=1))),(Item=(Type=Bone,Name=\"base_sholderpad_l\"),Transform=(Rotation=(X=0.123,Y=-0.0277,Z=0.0169,W=0.9919),Translation= ...(truncated)"
    WeightMinimum=0
    Weight=RigVMVariableNode() {variable=ClavUp_Emote}.Value
   )
   > Deform_Belt_1 = RigVMFunctionReferenceNode()
   > Deform_Lapel_1 = RigVMFunctionReferenceNode()
   > Deform_Armpit_1 = RigVMFunctionReferenceNode()
   n1 = &RigUnit_SphericalPoseReader(
    ActiveRegionScaleFactors=(PositiveWidth=1,NegativeWidth=1,PositiveHeight=1,NegativeHeight=1)
    ActiveRegionSize=0.05
    Debug=(bDrawDebug=True,bDraw2D=false,bDrawLocalAxes=False,DebugScale=25,DebugSegments=20,DebugThickness=0.25)
    DriverAxis=(X=-1,Y=0,Z=0)
    DriverItem=(Type=Bone,Name="clavicle_r")
    FalloffRegionScaleFactors=(PositiveWidth=1,NegativeWidth=1,PositiveHeight=1,NegativeHeight=1)
    FalloffSize=0.42
    FlipHeightScaling=False
    FlipWidthScaling=False
    OptionalParentItem=(Type=Bone,Name="None")
    RotationOffset=(X=180,Y=-25,Z=0)
    ExecutePin=@RigVMFunction_Sequence.A
   )
   > RigVMVariableNode(Value=@n1.OutputParam) {variable=ClavUp_R}
   RigVMFunction_MathBoolToFloat = &RigVMFunction_MathBoolToFloat(Value=RigVMVariableNode() {variable=Parachuting}.Value)
   Multiply = &RigVMFunction_MathDoubleMul(B=@RigVMFunction_MathBoolToFloat.Result, A=RigVMVariableNode() {variable=ClavUp_R}.Value)
   > RigVMVariableNode(Value=@Multiply.Result) {variable=ClavUp_Gliding}
   n2 = RigVMVariableNode() {variable=PlayingEmote_cust}
   > RigVMVariableNode(Value=&RigVMFunction_MathDoubleMul(B=&RigVMFunction_MathBoolToFloat(Value=@n2.Value).Result, A=RigVMVariableNode() {variable=ClavUp_R}.Value).Result) {variable=ClavUp_Emote}
   GreaterEqual = /Script/RigVM.RigVMFunction_MathDoubleGreaterEqual(B=0.1, A=RigVMVariableNode() {variable=ClavUp_R}.Value) {is=struct RigVMUnitNode, template="GreaterEqual::Execute(in A,in B,out Result)"}
   > RigVMVariableNode(Value=@GreaterEqual.Result) {variable=ClavUp_R_bool}
   RigVMFunction_MathBoolOr = &RigVMFunction_MathBoolOr(A=RigVMVariableNode() {variable=Parachuting}.Value, B=RigVMVariableNode() {variable=PlayingEmote_cust}.Value)
   RigVMFunction_MathBoolAnd = &RigVMFunction_MathBoolAnd(B=@RigVMFunction_MathBoolOr.Result, A=RigVMVariableNode() {variable=ClavUp_R_bool}.Value)
   > RigVMVariableNode(Value=@RigVMFunction_MathBoolAnd.Result) {variable=isHangGliding}
   DISPATCH_RigDispatch_GetAnimAttribute = RigVMDispatchNode(BoneName=root, Default=0, Name=deformRig_l_thigh_up_alpha) {template="DISPATCH_RigDispatch_GetAnimAttribute(in Name,in BoneName,in Default,out Value,out Found)"}
   > &RigUnit_ModifyTransforms(
    ItemToModify=((Item=(Type=Bone,Name="deform_glute_down_r"),Transform=(Rotation=(X=0,Y=0,Z=0,W=1),Translation=(X=-3,Y=0,Z=0),Scale3D=(X=1,Y=1,Z=1))))
    WeightMinimum=0
    Weight=@DISPATCH_RigDispatch_GetAnimAttribute.Value
   )
   &RigUnit_SphericalPoseReader(
    ActiveRegionScaleFactors=(PositiveWidth=1,NegativeWidth=1,PositiveHeight=1,NegativeHeight=1)
    ActiveRegionSize=0.2
    Debug=(bDrawDebug=false,bDraw2D=false,bDrawLocalAxes=False,DebugScale=25,DebugSegments=20,DebugThickness=0.25)
    DriverAxis=(X=-1,Y=0,Z=0)
    DriverItem=(Type=Bone,Name="clavicle_r")
    ExecutePin=()
    FalloffRegionScaleFactors=(PositiveWidth=1,NegativeWidth=1,PositiveHeight=1,NegativeHeight=1)
    FalloffSize=0.3
    FlipHeightScaling=False
    FlipWidthScaling=False
    OptionalParentItem=(Type=Bone,Name="None")
    RotationOffset=(X=180,Y=-45,Z=0)
   )
   RigVMVariableNode() {variable=PlayingEmote}
   RigVMVariableNode() {variable=PlayingEmote}
  ]}
]
hierarchy [
 root {loc=<0 0 0>, rot=<0 0 0 1>}
 pelvis {parent=root, loc=<0 -0.5301 98.0571>, rot=<0.0127 0.707 -0.0127 -0.707>}
 spine_01 {parent=pelvis, loc=<6.1208 -0.9328 0>, rot=<0 0 0.034 -0.9994>}
 spine_02 {parent=spine_01, loc=<6.5996 -0.2095 0>, rot=<0 0 0.0238 -0.9997>}
 spine_03 {parent=spine_02, loc=<6.4914 -0.1237 0>, rot=<0 0 0.0149 -0.9999>}
 spine_04 {parent=spine_03, loc=<6.9598 0.0794 0>, rot=<0 0 -0.0132 -0.9999>}
 spine_05 {parent=spine_04, loc=<6.9987 0 0>, rot=<0 0 -0.0015 -1>}
 attach_cape {parent=spine_05, loc=<9.0146 -11.7137 0.0002>, rot=<0.0616 -0.7044 -0.0616 -0.7044>}
 neck_01 {parent=spine_05, loc=<11.3138 0.1529 0>, rot=<0 0 -0.2003 -0.9797>}
 neck_02 {parent=neck_01, loc=<4.5627 0 0>, rot=<0 0 0.0504 -0.9987>}
 head {parent=neck_02, loc=<5.3914 0 0>, rot=<0 0 0.115 -0.9934>}
 attach_backpack {parent=spine_05, loc=<-0.1261 -13.4727 0.0003>, rot=<-0.0092 -0.707 0.0093 -0.707>}
 clavicle_l {parent=spine_05, loc=<5.4831 -0.1053 -3.2946>, rot=<0.0465 -0.7289 -0.009 -0.6829>}
 upperarm_l {parent=clavicle_l, loc=<9.8658 -1.2136 -0.4105>, rot=<0.0164 -0.3245 -0.0269 -0.9454>}
 lowerarm_l {parent=upperarm_l, loc=<26.465 -0.8345 0.259>, rot=<0.0036 0.0161 -0.2298 -0.9731>}
 hand_l {parent=lowerarm_l, loc=<22.7355 -0.3203 -0.6105>, rot=<-0.6784 -0.0223 0.0735 -0.7307>}
 thumb_01_l {parent=hand_l, loc=<1.3118 -0.3191 -2.459>, rot=<0.6546 -0.4042 0.0664 -0.6354>}
 thumb_02_l {parent=thumb_01_l, loc=<5.9336 -0.0448 0.0401>, rot=<-0.0037 0.0213 0.298 -0.9543>}
 thumb_03_l {parent=thumb_02_l, loc=<2.2223 0.3682 -0.0907>, rot=<0.0118 -0.013 0.0066 -0.9998>}
 index_metacarpal_l {parent=hand_l, loc=<1.6535 0.0183 -2.96>, rot=<0.0158 -0.0124 -0.0173 -0.9996>}
 index_01_l {parent=index_metacarpal_l, loc=<8.3008 0.0357 -0.0524>, rot=<-0.0044 0.0485 0.0924 -0.9945>}
 index_02_l {parent=index_01_l, loc=<3.3587 -0.3004 0.0452>, rot=<-0.002 0.0027 0.3225 -0.9466>}
 index_03_l {parent=index_02_l, loc=<2.5578 -0.123 -0.0252>, rot=<-0.002 0.0067 0.2558 -0.9667>}
 middle_metacarpal_l {parent=hand_l, loc=<2.1856 -0.1725 -0.5984>, rot=<0.006 0.0168 -0.0329 -0.9993>}
 middle_01_l {parent=middle_metacarpal_l, loc=<7.3275 -0.2447 0.0129>, rot=<-0.0208 0.0042 0.0858 -0.9961>}
 middle_02_l {parent=middle_01_l, loc=<4.0123 0.1401 0.1175>, rot=<-0.0218 0.0258 0.3512 -0.9357>}
 middle_03_l {parent=middle_02_l, loc=<2.6077 0.0534 0.1373>, rot=<0.0082 0.0188 0.1708 -0.9851>}
 ring_metacarpal_l {parent=hand_l, loc=<2.2074 -0.3625 1.3638>, rot=<-0.0284 0.0722 -0.0458 -0.9959>}
 ring_01_l {parent=ring_metacarpal_l, loc=<7.0295 -0.6197 -0.1446>, rot=<-0.0489 -0.0252 0.1892 -0.9804>}
 ring_02_l {parent=ring_01_l, loc=<3.6174 0.4647 -0.1027>, rot=<0 0 0.2806 -0.9598>}
 ring_03_l {parent=ring_02_l, loc=<2.4152 0.114 0.0199>, rot=<0 0 0.1189 -0.9929>}
 pinky_metacarpal_l {parent=hand_l, loc=<1.7616 -0.9753 3.3814>, rot=<-0.0397 0.0814 -0.069 -0.9935>}
 pinky_01_l {parent=pinky_metacarpal_l, loc=<7.1709 -0.4831 -0.1094>, rot=<-0.0654 -0.0035 0.2242 -0.9723>}
 pinky_02_l {parent=pinky_01_l, loc=<2.8274 0.0893 -0.1043>, rot=<0.0037 0.0011 0.2854 -0.9584>}
 pinky_03_l {parent=pinky_02_l, loc=<2.2046 0.0307 -0.0078>, rot=<-0.0002 0.0098 0.1445 -0.9895>}
 deform_wrist_dn_base_l {parent=hand_l, loc=<0 -1.5326 0>, rot=<0 0 0 -1>}
 deform_wrist_dn_l {parent=deform_wrist_dn_base_l, loc=<0 0 0>, rot=<0 0 0 1>}
 deform_wrist_up_base_l {parent=hand_l, loc=<0 2.3009 0>, rot=<0 0 0 -1>}
 deform_wrist_up_l {parent=deform_wrist_up_base_l, loc=<0 0 0>, rot=<0 0 0 1>}
 weapon_l {parent=hand_l, loc=<-0.2102 -3.6566 0.7616>, rot=<-0.0735 -0.2525 0.2602 -0.9291>}
 lowerarm_twist_02_l {parent=lowerarm_l, loc=<10.6707 0 0>, rot=<0 0 0 1>}
 lowerarm_twist_01_l {parent=lowerarm_l, loc=<18.316 0 0>, rot=<0 0 0 1>}
 deform_elbow_in_base_l {parent=lowerarm_l, loc=<0.1834 3.6062 -0.2572>, rot=<0 0 0 -1>}
 deform_elbow_in_l {parent=deform_elbow_in_base_l, loc=<0 0 0>, rot=<0 0 0 1>}
 deform_elbow_out_base_l {parent=lowerarm_l, loc=<0 -4 0>, rot=<0 0 0 1>}
 deform_elbow_out_l {parent=deform_elbow_out_base_l, loc=<0 0 0>, rot=<0 0 0 1>}
 upperarm_twist_01_l {parent=upperarm_l, loc=<7.2135 -0.1134 0>, rot=<0 0 0 1>}
 deform_upperarm_bk_base_l {parent=upperarm_twist_01_l, loc=<-6.4053 -4.3746 0.314>, rot=<0 0 0 -1>}
 deform_upperarm_bk_l {parent=deform_upperarm_bk_base_l, loc=<0 0 0>, rot=<0 0 0 1>}
 deform_upperarm_fr_base_l {parent=upperarm_twist_01_l, loc=<-5.6312 5.092 0.3354>, rot=<0 0 0 -1>}
 deform_upperarm_fr_l {parent=deform_upperarm_fr_base_l, loc=<0 0 0>, rot=<0 0 0 1>}
 deform_upperarm_up_base_l {parent=upperarm_twist_01_l, loc=<-8.891 0.4825 3.9517>, rot=<0 0 0 -1>}
 deform_upperarm_up_l {parent=deform_upperarm_up_base_l, loc=<0 0 0>, rot=<0 0 0 1>}
 base_upperarm_inner_l {parent=upperarm_twist_01_l, loc=<1.2595 0.1134 -4.5435>, rot=<0 0 0 -1>}
 dyn_upperarm_inner_l {parent=base_upperarm_inner_l, loc=<0 0 0>, rot=<0 0 0 -1>}
 upperarm_twist_02_l {parent=upperarm_l, loc=<19.7998 -0.1134 0>, rot=<0 0 0 1>}
 base_sholderpad_l {parent=clavicle_l, loc=<8.4991 -1.1726 5.4715>, rot=<0.0277 -0.7537 -0.0151 -0.6564>}
 base_sholderpad_b_l {parent=base_sholderpad_l, loc=<2.7331 -6.0883 3.8476>, rot=<0 0 0 1>}
 dyn_sholderpad_b_l {parent=base_sholderpad_b_l, loc=<0 0 0>, rot=<0 0 0 1>}
 base_sholderpad_f_l {parent=base_sholderpad_l, loc=<2.4015 5.7887 3.9232>, rot=<0 0 0 1>}
 dyn_sholderpad_f_l {parent=base_sholderpad_f_l, loc=<0 0 0>, rot=<0 0 0 1>}
 base_sholderpad_m_l {parent=base_sholderpad_l, loc=<-0.3497 -0.0417 3.8669>, rot=<0 0 0 1>}
 dyn_sholderpad_m_l {parent=base_sholderpad_m_l, loc=<0 0 0>, rot=<0 0 0 1>}
 clavicle_r {parent=spine_05, loc=<5.4831 -0.1053 3.2946>, rot=<0.7289 0.0465 -0.6829 0.009>}
 upperarm_r {parent=clavicle_r, loc=<-9.8658 1.2136 0.4105>, rot=<0.0164 -0.3245 -0.0269 -0.9454>}
 lowerarm_r {parent=upperarm_r, loc=<-26.465 0.8345 -0.259>, rot=<0.0036 0.0161 -0.2298 -0.9731>}
 hand_r {parent=lowerarm_r, loc=<-22.7355 0.3203 0.6106>, rot=<-0.6784 -0.0223 0.0735 -0.7307>}
 thumb_01_r {parent=hand_r, loc=<-1.3118 0.3191 2.459>, rot=<0.6546 -0.4042 0.0664 -0.6354>}
 thumb_02_r {parent=thumb_01_r, loc=<-5.9336 0.0448 -0.0401>, rot=<-0.0037 0.0213 0.298 -0.9543>}
 thumb_03_r {parent=thumb_02_r, loc=<-2.2223 -0.3682 0.0907>, rot=<0.0118 -0.013 0.0066 -0.9998>}
 index_metacarpal_r {parent=hand_r, loc=<-1.6535 -0.0183 2.96>, rot=<0.0158 -0.0124 -0.0173 -0.9996>}
 index_01_r {parent=index_metacarpal_r, loc=<-8.3008 -0.0357 0.0524>, rot=<-0.0044 0.0485 0.0924 -0.9945>}
 index_02_r {parent=index_01_r, loc=<-3.3587 0.3004 -0.0452>, rot=<-0.002 0.0027 0.3225 -0.9466>}
 index_03_r {parent=index_02_r, loc=<-2.5578 0.123 0.0252>, rot=<-0.002 0.0067 0.2558 -0.9667>}
 middle_metacarpal_r {parent=hand_r, loc=<-2.1856 0.1725 0.5984>, rot=<0.006 0.0168 -0.0329 -0.9993>}
 middle_01_r {parent=middle_metacarpal_r, loc=<-7.3275 0.2447 -0.0129>, rot=<-0.0208 0.0042 0.0858 -0.9961>}
 middle_02_r {parent=middle_01_r, loc=<-4.0123 -0.1401 -0.1175>, rot=<-0.0218 0.0258 0.3512 -0.9357>}
 middle_03_r {parent=middle_02_r, loc=<-2.6077 -0.0534 -0.1373>, rot=<0.0082 0.0188 0.1708 -0.9851>}
 ring_metacarpal_r {parent=hand_r, loc=<-2.2074 0.3625 -1.3638>, rot=<-0.0284 0.0722 -0.0458 -0.9959>}
 ring_01_r {parent=ring_metacarpal_r, loc=<-7.0295 0.6197 0.1446>, rot=<-0.0489 -0.0252 0.1892 -0.9804>}
 ring_02_r {parent=ring_01_r, loc=<-3.6174 -0.4647 0.1027>, rot=<0 0 0.2806 -0.9598>}
 ring_03_r {parent=ring_02_r, loc=<-2.4152 -0.114 -0.0199>, rot=<0 0 0.1189 -0.9929>}
 pinky_metacarpal_r {parent=hand_r, loc=<-1.7616 0.9753 -3.3814>, rot=<-0.0397 0.0814 -0.069 -0.9935>}
 pinky_01_r {parent=pinky_metacarpal_r, loc=<-7.1709 0.4831 0.1094>, rot=<-0.0654 -0.0035 0.2242 -0.9723>}
 pinky_02_r {parent=pinky_01_r, loc=<-2.8274 -0.0893 0.1043>, rot=<0.0037 0.0011 0.2854 -0.9584>}
 pinky_03_r {parent=pinky_02_r, loc=<-2.2046 -0.0307 0.0078>, rot=<-0.0002 0.0098 0.1445 -0.9895>}
 deform_wrist_dn_base_r {parent=hand_r, loc=<-0.0001 1.5328 0>, rot=<0 0 0 -1>}
 deform_wrist_dn_r {parent=deform_wrist_dn_base_r, loc=<0 0 0>, rot=<0 0 0 1>}
 deform_wrist_up_base_r {parent=hand_r, loc=<-0.0002 -2.3007 -0.0001>, rot=<0 0 0 -1>}
 deform_wrist_up_r {parent=deform_wrist_up_base_r, loc=<0 0 0>, rot=<0 0 0 1>}
 weapon_r {parent=hand_r, loc=<0.4809 3.7076 -1.326>, rot=<-0.0003 -0.008 0.132 -0.9912>}
 lowerarm_twist_02_r {parent=lowerarm_r, loc=<-10.6707 0 0>, rot=<0 0 0 -1>}
 lowerarm_twist_01_r {parent=lowerarm_r, loc=<-18.316 0 0>, rot=<0 0 0 -1>}
 deform_elbow_in_base_r {parent=lowerarm_r, loc=<-0.1834 -3.6062 0.2572>, rot=<0 0 0 -1>}
 deform_elbow_in_r {parent=deform_elbow_in_base_r, loc=<0 0 0>, rot=<0 0 0 1>}
 deform_elbow_out_base_r {parent=lowerarm_r, loc=<0 4 -0.0001>, rot=<0 0 0 -1>}
 deform_elbow_out_r {parent=deform_elbow_out_base_r, loc=<0 0 0>, rot=<0 0 0 1>}
 upperarm_twist_01_r {parent=upperarm_r, loc=<-7.2135 0.1134 0>, rot=<0 0 0 -1>}
 deform_upperarm_bk_base_r {parent=upperarm_twist_01_r, loc=<6.4053 4.3746 -0.314>, rot=<0 0 0 -1>}
 deform_upperarm_bk_r {parent=deform_upperarm_bk_base_r, loc=<0 0 0>, rot=<0 0 0 1>}
 deform_upperarm_fr_base_r {parent=upperarm_twist_01_r, loc=<5.6312 -5.092 -0.3354>, rot=<0 0 0 -1>}
 deform_upperarm_fr_r {parent=deform_upperarm_fr_base_r, loc=<0 0 0>, rot=<0 0 0 1>}
 deform_upperarm_up_base_r {parent=upperarm_twist_01_r, loc=<8.891 -0.4825 -3.9517>, rot=<0 0 0 -1>}
 deform_upperarm_up_r {parent=deform_upperarm_up_base_r, loc=<0 0 0>, rot=<0 0 0 1>}
 base_upperarm_inner_r {parent=upperarm_twist_01_r, loc=<-1.2597 -0.1134 4.5437>, rot=<0 0 0 -1>}
 dyn_upperarm_inner_r {parent=base_upperarm_inner_r, loc=<0 0 0>, rot=<0 0 0 1>}
 upperarm_twist_02_r {parent=upperarm_r, loc=<-19.7998 0.1134 0>, rot=<0 0 0 -1>}
 base_sholderpad_r {parent=clavicle_r, loc=<-8.499 1.1726 -5.4713>, rot=<0.0277 -0.7537 -0.0151 -0.6564>}
 base_sholderpad_f_r {parent=base_sholderpad_r, loc=<-2.4015 -5.7887 -3.9233>, rot=<0 0 0 1>}
 dyn_sholderpad_f_r {parent=base_sholderpad_f_r, loc=<0 0 0>, rot=<0 0 0 1>}
 base_sholderpad_m_r {parent=base_sholderpad_r, loc=<0.3499 0.0417 -3.8669>, rot=<0 0 0 1>}
 dyn_sholderpad_m_r {parent=base_sholderpad_m_r, loc=<0 0 0>, rot=<0 0 0 1>}
 base_sholderpad_b_r {parent=base_sholderpad_r, loc=<-2.733 6.0883 -3.8477>, rot=<0 0 0 1>}
 dyn_sholderpad_b_r {parent=base_sholderpad_b_r, loc=<0 0 0>, rot=<0 0 0 1>}
 base_lower_lapel_r {parent=spine_05, loc=<3.3065 7.6778 6.7001>, rot=<0.212 0.554 0.1485 -0.7912>}
 dyn_lower_lapel_r {parent=base_lower_lapel_r, loc=<0 0 0>, rot=<0 0 0 1>}
 base_upper_lapel_r {parent=spine_05, loc=<9.5849 3.204 6.4146>, rot=<0.212 0.554 0.1485 -0.7912>}
 dyn_upper_lapel_r {parent=base_upper_lapel_r, loc=<0 0 0>, rot=<0 0 0 1>}
 base_lower_lapel_l {parent=spine_05, loc=<3.3065 7.6778 -6.7>, rot=<-0.1825 -0.5246 0.1163 -0.8234>}
 dyn_lower_lapel_l {parent=base_lower_lapel_l, loc=<0 0 0>, rot=<0 0 0 1>}
 base_upper_lapel_l {parent=spine_05, loc=<9.5849 3.204 -6.415>, rot=<-0.1825 -0.5246 0.1163 -0.8234>}
 dyn_upper_lapel_l {parent=base_upper_lapel_l, loc=<0 0 0>, rot=<0 0 0 1>}
 base_shirtmover_l {parent=spine_05, loc=<-1.739 -1.7347 -12.9447>, rot=<0.0465 -0.7289 -0.009 -0.6829>}
 shirtmover_l {parent=base_shirtmover_l, loc=<0 0 0>, rot=<0 0 0 1>}
 base_shirtmover_r {parent=spine_05, loc=<-1.8036 -1.7917 12.9371>, rot=<0.7289 0.0465 -0.6829 0.009>}
 shirtmover_r {parent=base_shirtmover_r, loc=<0 0 0>, rot=<0 0 0 1>}
 deform_pec_aim_base_l {parent=spine_04, loc=<3.2456 9.3309 -1.2237>, rot=<0.0603 -0.4033 0.3318 -0.8507>}
 deform_pec_aim_l {parent=deform_pec_aim_base_l, loc=<0 0 0>, rot=<0 0 0 1>}
 deform_pec_base_l {parent=deform_pec_aim_l, loc=<9.3665 0 0>, rot=<0 0 0 1>}
 deform_pec_l {parent=deform_pec_base_l, loc=<0 0 0>, rot=<0 0 0 1>}
 deform_pec_aim_base_r {parent=spine_04, loc=<3.2452 9.3309 1.2237>, rot=<-0.4033 -0.0603 0.8507 0.3318>}
 deform_pec_aim_r {parent=deform_pec_aim_base_r, loc=<0 0 0>, rot=<0 0 0 1>}
 deform_pec_base_r {parent=deform_pec_aim_r, loc=<-9.3667 -0.0003 -0.0004>, rot=<0 0 0 1>}
 deform_pec_r {parent=deform_pec_base_r, loc=<0 0 0>, rot=<0 0 0 1>}
 thigh_l {parent=pelvis, loc=<-1.0855 -0.2109 -8.7053>, rot=<0.0009 0.0355 0.0155 -0.9993>}
 calf_l {parent=thigh_l, loc=<-42.5284 -0.0019 -0.7297>, rot=<0.0006 -0.0161 -0.0391 -0.9991>}
 foot_l {parent=calf_l, loc=<-46.7401 -2.5014 -0.592>, rot=<0.0427 -0.0238 0.0353 -0.9982>}
 ball_l {parent=foot_l, loc=<-3.5658 10.8066 -0.1544>, rot=<0.0001 -0.0001 -0.7063 -0.7079>}
 calf_twist_02_l {parent=calf_l, loc=<-19.1546 -1.7943 -0.2638>, rot=<0.0237 0.0095 -0.0035 -0.9997>}
 calf_twist_01_l {parent=calf_l, loc=<-35.8483 -1.9197 -0.5794>, rot=<0.0237 0.0095 -0.0035 -0.9997>}
 deform_knee_in_base_l {parent=calf_l, loc=<0.8031 -5.0231 0.2456>, rot=<0 0 0 -1>}
 deform_knee_in_l {parent=deform_knee_in_base_l, loc=<0 0 0>, rot=<0 0 0 1>}
 deform_knee_out_base_l {parent=calf_l, loc=<0 4 0>, rot=<0 0 0 -1>}
 deform_knee_out_l {parent=deform_knee_out_base_l, loc=<0 0 0>, rot=<0 0 0 1>}
 base_lace_l {parent=calf_l, loc=<-32.3892 2.4555 -1.557>, rot=<0.0149 -0.0476 0.9981 0.0371>}
 dyn_lace_l_loop_r {parent=base_lace_l, loc=<-0.0073 -0.1156 1.2725>, rot=<0 0 0 1>}
 dyn_lace_l_loop_l {parent=base_lace_l, loc=<0.0072 0.1143 -1.2579>, rot=<0 0 0 1>}
 dyn_lace_l_strand_r {parent=base_lace_l, loc=<1.522 -0.1076 1.0137>, rot=<0 0 0 1>}
 dyn_lace_l_strand_l {parent=base_lace_l, loc=<1.5792 0.0735 -0.985>, rot=<0 0 0 1>}
 thigh_twist_01_l {parent=thigh_l, loc=<-27.952 0 0>, rot=<0.0013 0 0 -1>}
 deform_groin_base_l {parent=thigh_l, loc=<0.4801 5.6776 -0.2486>, rot=<0 0 0 -1>}
 deform_groin_l {parent=deform_groin_base_l, loc=<0 0 0>, rot=<0 0 0 1>}
 thigh_r {parent=pelvis, loc=<-1.0855 -0.2109 8.7053>, rot=<0.0355 -0.0009 0.9993 0.0155>}
 calf_r {parent=thigh_r, loc=<42.5284 0.0019 0.7297>, rot=<0.0006 -0.0161 -0.0391 -0.9991>}
 foot_r {parent=calf_r, loc=<46.7401 2.5014 0.592>, rot=<0.0427 -0.0238 0.0353 -0.9982>}
 ball_r {parent=foot_r, loc=<3.5658 -10.8066 0.1544>, rot=<0.0001 -0.0001 -0.7063 -0.7079>}
 calf_twist_02_r {parent=calf_r, loc=<19.1546 1.7943 0.2638>, rot=<0.0237 0.0095 -0.0035 -0.9997>}
 calf_twist_01_r {parent=calf_r, loc=<35.8483 1.9197 0.5794>, rot=<0.0237 0.0095 -0.0035 -0.9997>}
 deform_knee_in_base_r {parent=calf_r, loc=<-0.8031 5.0231 -0.2457>, rot=<0 0 0 -1>}
 deform_knee_in_r {parent=deform_knee_in_base_r, loc=<0 0 0>, rot=<0 0 0 1>}
 deform_knee_out_base_r {parent=calf_r, loc=<0 -4 0>, rot=<0 0 0 -1>}
 deform_knee_out_r {parent=deform_knee_out_base_r, loc=<0 0 0>, rot=<0 0 0 1>}
 base_lace_r {parent=calf_r, loc=<32.3675 -2.5369 1.2832>, rot=<0.0427 -0.0238 0.0353 -0.9982>}
 dyn_lace_r_loop_r {parent=base_lace_r, loc=<-0.0073 -0.1156 1.2725>, rot=<0 0 0 1>}
 dyn_lace_r_loop_l {parent=base_lace_r, loc=<0.0072 0.1143 -1.2579>, rot=<0 0 0 1>}
 dyn_lace_r_strand_r {parent=base_lace_r, loc=<1.522 -0.1076 1.0137>, rot=<0 0 0 1>}
 dyn_lace_r_strand_l {parent=base_lace_r, loc=<1.5792 0.0735 -0.985>, rot=<0 0 0 1>}
 thigh_twist_01_r {parent=thigh_r, loc=<27.952 0 0>, rot=<0.0013 0 0 -1>}
 deform_groin_base_r {parent=thigh_r, loc=<-0.4801 -5.6776 0.2486>, rot=<0 0 0 -1>}
 deform_groin_r {parent=deform_groin_base_r, loc=<0 0 0>, rot=<0 0 0 1>}
 deform_glute_base_l {parent=pelvis, loc=<-7.3466 -9.0213 -9.1245>, rot=<0.0172 0.031 -0.4477 -0.8935>}
 deform_glute_l {parent=deform_glute_base_l, loc=<0 0 0>, rot=<0 0 0 1>}
 deform_glute_base_r {parent=pelvis, loc=<-7.3464 -9.0213 9.1245>, rot=<0.031 -0.0172 0.8935 -0.4477>}
 deform_glute_r {parent=deform_glute_base_r, loc=<0 0 0>, rot=<0 0 0 1>}
 base_cord {parent=pelvis, loc=<-2.3075 -4.1843 -16.071>, rot=<-0.0735 -0.5353 0.8402 -0.0468>}
 dyn_cord {parent=base_cord, loc=<0 0 0>, rot=<0 0 0 1>}
 base_auxcord_mic {parent=dyn_cord, loc=<-3.1155 0.3239 -3.1573>, rot=<-0.8192 0 0 -0.5736>}
 dyn_auxcord_mic_1 {parent=base_auxcord_mic, loc=<0 0 0>, rot=<0 0 0 1>}
 dyn_auxcord_mic_2 {parent=dyn_auxcord_mic_1, loc=<4.8277 -0.2226 -1.0055>, rot=<0 0 0 1>}
 base_auxcord_front_loop {parent=dyn_cord, loc=<3.8023 -2.8681 -5.9236>, rot=<-0.3007 0 0 -0.9537>}
 dyn_auxcord_front_loop_1 {parent=base_auxcord_front_loop, loc=<0 0 0>, rot=<0 0 0 1>}
 dyn_auxcord_front_loop_2 {parent=dyn_auxcord_front_loop_1, loc=<8.7561 -0.0074 -0.5989>, rot=<0 0 0 1>}
 dyn_auxcord_back_loop {parent=dyn_cord, loc=<-2.4829 1.4089 2.0926>, rot=<-0.7071 0 0 -0.7071>}
 dyn_auxcord_back_loop_1 {parent=dyn_auxcord_back_loop, loc=<0 0 0>, rot=<0 0 0 1>}
 dyn_auxcord_back_loop_2 {parent=dyn_auxcord_back_loop_1, loc=<13.0041 0.009 1.3836>, rot=<0 0 0 1>}
 base_auxcord_end {parent=dyn_cord, loc=<2.4022 -1.0977 5.7402>, rot=<0.9029 -0.079 0.0368 0.421>}
 dyn_auxcord_end {parent=base_auxcord_end, loc=<0 0 0>, rot=<0 0 0 1>}
 belt_helper_r {parent=pelvis, loc=<3.7856 7.478 10.3015>, rot=<0.0445 -0.5543 0.8203 0.1341>}
 belt_helper_l {parent=pelvis, loc=<-0.2417 8.7858 -10.085>, rot=<0.0445 -0.5543 0.8203 0.1341>}
 deform_glute_down_r {parent=pelvis, loc=<-12.6802 -1.07 9.7129>, rot=<0.0759 -0.0008 0.9957 0.0528>}
 upperarm_r_ctrl {controlType=EULER_TRANSFORM, type=Control, loc=<-13.0371 -5.8204 135.5446>, rot=<0.9343 -0.0039 0.3563 0.0065>}
 upperarm_l_ctrl {controlType=EULER_TRANSFORM, type=Control, loc=<13.0371 -5.8204 135.5446>, rot=<-0.0065 0.3563 0.0039 0.9343>}
]
variables [
 ArmpitFix {type=double, default=0}
 BeltFix {type=double, default=1}
 ClavUp_Emote {type=double, default=0}
 ClavUp_Gliding {type=double, default=0}
 ClavUp_R {type=double, default=0}
 ClavUp_R_bool {type=bool, default=False}
 IsNotSkydive {type=bool, default=False}
 NewVar {type=double, default=0}
 ParachuteEmote {type=double, default=0}
 Parachuting {type=bool, default=False}
 PlayingEmote {type=bool, default=False}
 PlayingEmote_cust {type=bool, default=False}
 ShoulderFix {type=double, default=0}
 icrouch {type=double, default=0}
 is skydive {type=double, default=0}
 isHangGliding {type=bool, default=False}
 parachute open or emote {type=bool, default=False}
]
