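# Physics asset for the lower lapels of the StrategicSpur_Mic female medium cosmetic body.
# Each dynamic lapel capsule is constrained to a tiny base capsule with a small swing limit on one axis, tighter on the right, and zero-degree limits on the others.
# Both upper arms and the right clavicle add capsule bodies.
# Only the right arm is constrained, with free angular motion to its clavicle; the left arm has no constraint.
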PhysicsAsset /BRCosmetics/Characters/Player/Female/Medium/Bodies/F_MED_StrategicSpur_Mic/Meshes/F_MED_StrategicSpur_Mic_Lapels_Physics
bodies {
 base_lower_lapel_l [{shapeType=Capsule, length=0.505, radius=0.505, shapeName=base_lower_lapel_l_capsule, rotation=<rot 0 -90 0>}]
 base_lower_lapel_r [{shapeType=Capsule, length=0.505, radius=0.505, shapeName=base_lower_lapel_r_capsule, rotation=<rot 0 -90 0>}]
 clavicle_r [{shapeType=Capsule, length=5, radius=6, shapeName=clavicle_r_capsule, center=<-5 0 0>, rotation=<rot 90 0 0>}]
 dyn_lower_lapel_l [{
   shapeType=Capsule
   length=12.4201
   radius=2.2198
   shapeName=dyn_lower_lapel_l_capsule
   center=<1.6774 -0.1093 -4.653>
   rotation=<rot 1.6908 -90 40.9844>
  }]
 dyn_lower_lapel_r [{
   shapeType=Capsule
   length=11.9127
   radius=2.2747
   shapeName=dyn_lower_lapel_r_capsule
   center=<2.0848 -0.8712 4.2032>
   rotation=<rot 0.7313 -89.1664 -40.9668>
  }]
 upperarm_l [{shapeType=Capsule, length=23.1008, radius=6, shapeName=upperarm_l_capsule, center=<12.0904 0.2315 0>, rotation=<rot 90 180 -175.4796>}]
 upperarm_r [{shapeType=Capsule, length=23.1008, radius=6, shapeName=upperarm_r_capsule, center=<-12.0904 0.2315 0>, rotation=<rot 90 180 -175.4796>}]
}
constraints [
 {bone1Name=dyn_lower_lapel_l, bone2Name=base_lower_lapel_l, swing1LimitDegrees=10, swing2LimitDegrees=0, twistLimitDegrees=0}
 {bone1Name=dyn_lower_lapel_r, bone2Name=base_lower_lapel_r, swing1LimitDegrees=2, swing2LimitDegrees=0, twistLimitDegrees=0}
 {
  bone1Name=upperarm_r
  bone2Name=clavicle_r
  swing1LimitDegrees=45
  swing1Motion=Free
  swing2LimitDegrees=45
  swing2Motion=Free
  twistLimitDegrees=45
  twistMotion=Free
 }
]
