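# Physics asset for the StrategicSpur Mic female medium body cosmetic, defining simulation bodies for its dangling aux cord and microphone.
# Three capsule chains (front loop, back loop and mic) branch from base_cord, each linked by constraints with limited swing and zero twist limit.
# The mic chain's outer segments have the largest capsules among the cord bodies and its tip swings widest.
# An unconstrained left thigh capsule is also included, likely as a collider for the cord.
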PhysicsAsset /BRCosmetics/Characters/Player/Female/Medium/Bodies/F_MED_StrategicSpur_Mic/Meshes/F_MED_StrategicSpur_Mic_Physics
bodies {
 base_auxcord_front_loop [{shapeType=Capsule, length=1, radius=1, shapeName=base_auxcord_front_loop_capsule, rotation=<rot 90 0 0>}]
 base_auxcord_mic [{shapeType=Capsule, length=1, radius=1, shapeName=base_auxcord_mic_capsule, rotation=<rot 90 0 0>}]
 base_cord [{shapeType=Capsule, length=1, radius=1, shapeName=base_cord_capsule, rotation=<rot 90 0 0>}]
 dyn_auxcord_back_loop [{shapeType=Capsule, length=1, radius=1, shapeName=dyn_auxcord_back_loop_capsule, rotation=<rot 90 0 0>}]
 dyn_auxcord_back_loop_1 [{shapeType=Capsule, length=3, radius=2.5, shapeName=dyn_auxcord_back_loop_1_capsule, center=<4 0 0>, rotation=<rot 90 0 0>}]
 dyn_auxcord_back_loop_2 [{
   shapeType=Capsule
   length=3.5931
   radius=2.5647
   shapeName=dyn_auxcord_back_loop_2_capsule
   center=<3.2383 -0.4835 0.1802>
   rotation=<rot 83.0275 -164.278 -150.4277>
  }]
 dyn_auxcord_front_loop_1 [{shapeType=Capsule, length=3, radius=1, shapeName=dyn_auxcord_front_loop_1_capsule, center=<4 0 -1.582>, rotation=<rot 90 0 0>}]
 dyn_auxcord_front_loop_2 [{
   shapeType=Capsule
   length=3.5931
   radius=2.4069
   shapeName=dyn_auxcord_front_loop_2_capsule
   center=<2 0.0533 -1.0593>
   rotation=<rot 90 0 0>
  }]
 dyn_auxcord_mic_1 [{shapeType=Capsule, length=4, radius=3, shapeName=dyn_auxcord_mic_1_capsule, center=<8 -1.0663 -0.0374>, rotation=<rot 90 0 0>}]
 dyn_auxcord_mic_2 [{shapeType=Capsule, length=5.5, radius=5.8431, shapeName=dyn_auxcord_mic_2_capsule, center=<18.076 -1.2398 0.7933>, rotation=<rot 90 0 0>}]
 thigh_l [{shapeType=Capsule, length=15, radius=8, shapeName=thigh_l_capsule, center=<-12 0 0>, rotation=<rot 90 0 0>}]
}
constraints [
 {bone1Name=base_auxcord_front_loop, bone2Name=base_cord, swing1LimitDegrees=20, swing2LimitDegrees=20, twistLimitDegrees=0}
 {bone1Name=dyn_auxcord_front_loop_1, bone2Name=base_auxcord_front_loop, swing1LimitDegrees=20, swing2LimitDegrees=25, twistLimitDegrees=0}
 {bone1Name=dyn_auxcord_front_loop_2, bone2Name=dyn_auxcord_front_loop_1, swing1LimitDegrees=30, swing2LimitDegrees=30, twistLimitDegrees=0}
 {bone1Name=base_auxcord_mic, bone2Name=base_cord, swing1LimitDegrees=20, swing2LimitDegrees=20, twistLimitDegrees=0}
 {bone1Name=dyn_auxcord_mic_1, bone2Name=base_auxcord_mic, swing1LimitDegrees=30, swing2LimitDegrees=30, twistLimitDegrees=0}
 {bone1Name=dyn_auxcord_mic_2, bone2Name=dyn_auxcord_mic_1, swing1LimitDegrees=60, swing2LimitDegrees=40, twistLimitDegrees=0}
 {bone1Name=dyn_auxcord_back_loop, bone2Name=base_cord, swing1LimitDegrees=20, swing2LimitDegrees=20, twistLimitDegrees=0}
 {bone1Name=dyn_auxcord_back_loop_1, bone2Name=dyn_auxcord_back_loop, swing1LimitDegrees=15, swing2LimitDegrees=15, twistLimitDegrees=0}
 {bone1Name=dyn_auxcord_back_loop_2, bone2Name=dyn_auxcord_back_loop_1, swing1LimitDegrees=20, swing2LimitDegrees=20, twistLimitDegrees=0}
]
