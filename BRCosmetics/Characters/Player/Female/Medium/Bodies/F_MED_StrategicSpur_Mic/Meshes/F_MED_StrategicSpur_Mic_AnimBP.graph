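# Post-process style AnimBlueprint for the F_MED_StrategicSpur_Mic cosmetic body, adding cloth, cord and lace motion to the pose copied from its parent mesh.
# A Control Rig and two RigidBody simulations feed corrective bone offsets for shoulder pads, mic cord, shirt and lapels.
# Anim Dynamics springs then move laces, upper lapels and cord end.
# The EventGraph derives fix weights from movement, parachute, emote and clavicle-rotation checks.
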AnimBlueprint /BRCosmetics/Characters/Player/Female/Medium/Bodies/F_MED_StrategicSpur_Mic/Meshes/F_MED_StrategicSpur_Mic_AnimBP
target_skeleton=&/F_MED_StrategicSpur_Mic_Skeleton
graphs [
 AnimGraph {nodes [
   n1 = Anim Dynamics - Bone: dyn_lace_r_loop_r!(ComponentPose=Local To Component(LocalPose=Use cached pose 'Fixes'()), Alpha=1, GravityOverride:Vector="0,0,-980") -> (Pose) {
    angularDampingOverride=0.99
    angularSpringConstant=5000
    bAngularSpring=true
    bOverrideAngularDamping=true
    bUseGravityOverride=true
    gravityScale=0.35
    simulationSpace=RootRelative
    alphaBoolBlend {blendOption=Linear}
    boundBone {boneName=dyn_lace_r_loop_r}
    componentLinearAccScale=<1 1 1>
    componentLinearVelScale=<1 1 1>
    gravityOverride {z=-980}
    physicsBodyDefinitions [{
      boundBone {boneName=dyn_lace_r_loop_r}
      boxExtents=<20 30 30>
      constraintSetup {angularLimitsMax=<15 15 15>, angularLimitsMin=<-15 -15 -15>, angularTarget=<1 0 0>}
      localJointOffset=<4 0 0>
     }]
   }
   n2 = Anim Dynamics - Bone: dyn_lace_r_loop_l!(ComponentPose=@n1, Alpha=1, GravityOverride:Vector="0,0,-980") -> (Pose) {
    angularDampingOverride=0.99
    angularSpringConstant=5000
    bAngularSpring=true
    bOverrideAngularDamping=true
    bUseGravityOverride=true
    gravityScale=0.35
    simulationSpace=RootRelative
    alphaBoolBlend {blendOption=Linear}
    boundBone {boneName=dyn_lace_r_loop_l}
    componentLinearAccScale=<1 1 1>
    componentLinearVelScale=<1 1 1>
    gravityOverride {z=-980}
    physicsBodyDefinitions [{
      boundBone {boneName=dyn_lace_r_loop_l}
      boxExtents=<20 30 30>
      constraintSetup {angularLimitsMax=<15 15 15>, angularLimitsMin=<-15 -15 -15>, angularTarget=<1 0 0>}
      localJointOffset=<4 0 0>
     }]
   }
   n3 = Anim Dynamics - Bone: dyn_lace_r_strand_r!(ComponentPose=@n2, Alpha=1, GravityOverride:Vector="0,0,-980") -> (Pose) {
    angularDampingOverride=0.99
    angularSpringConstant=5000
    bAngularSpring=true
    bOverrideAngularDamping=true
    bUseGravityOverride=true
    gravityScale=0.35
    simulationSpace=RootRelative
    alphaBoolBlend {blendOption=Linear}
    boundBone {boneName=dyn_lace_r_strand_r}
    componentLinearAccScale=<1 1 1>
    componentLinearVelScale=<1 1 1>
    gravityOverride {z=-980}
    physicsBodyDefinitions [{
      boundBone {boneName=dyn_lace_r_strand_r}
      boxExtents=<20 30 30>
      constraintSetup {angularLimitsMax=<15 15 15>, angularLimitsMin=<-15 -15 -15>, angularTarget=<1 0 0>}
      localJointOffset=<4 0 0>
     }]
   }
   n4 = Anim Dynamics - Bone: dyn_lace_r_strand_l!(ComponentPose=@n3, Alpha=1, GravityOverride:Vector="0,0,-980") -> (Pose) {
    angularDampingOverride=0.99
    angularSpringConstant=5000
    bAngularSpring=true
    bOverrideAngularDamping=true
    bUseGravityOverride=true
    gravityScale=0.35
    simulationSpace=RootRelative
    alphaBoolBlend {blendOption=Linear}
    boundBone {boneName=dyn_lace_r_strand_l}
    componentLinearAccScale=<1 1 1>
    componentLinearVelScale=<1 1 1>
    gravityOverride {z=-980}
    physicsBodyDefinitions [{
      boundBone {boneName=dyn_lace_r_strand_l}
      boxExtents=<20 30 30>
      constraintSetup {angularLimitsMax=<15 15 15>, angularLimitsMin=<-15 -15 -15>, angularTarget=<1 0 0>}
      localJointOffset=<4 0 0>
     }]
   }
   n5 = Anim Dynamics - Bone: dyn_lace_l_strand_l!(ComponentPose=@n4, Alpha=1, GravityOverride:Vector="0,0,-980") -> (Pose) {
    angularDampingOverride=0.99
    angularSpringConstant=5000
    bAngularSpring=true
    bOverrideAngularDamping=true
    bUseGravityOverride=true
    gravityScale=0.35
    simulationSpace=RootRelative
    alphaBoolBlend {blendOption=Linear}
    boundBone {boneName=dyn_lace_l_strand_l}
    componentLinearAccScale=<1 1 1>
    componentLinearVelScale=<1 1 1>
    gravityOverride {z=-980}
    physicsBodyDefinitions [{
      boundBone {boneName=dyn_lace_l_strand_l}
      boxExtents=<20 30 30>
      constraintSetup {angularLimitsMax=<15 15 15>, angularLimitsMin=<-15 -15 -15>, angularTarget=<1 0 0>}
      localJointOffset=<4 0 0>
     }]
   }
   n6 = Anim Dynamics - Bone: dyn_lace_l_strand_r!(ComponentPose=@n5, Alpha=1, GravityOverride:Vector="0,0,-980") -> (Pose) {
    angularDampingOverride=0.99
    angularSpringConstant=5000
    bAngularSpring=true
    bOverrideAngularDamping=true
    bUseGravityOverride=true
    gravityScale=0.35
    simulationSpace=RootRelative
    alphaBoolBlend {blendOption=Linear}
    boundBone {boneName=dyn_lace_l_strand_r}
    componentLinearAccScale=<1 1 1>
    componentLinearVelScale=<1 1 1>
    gravityOverride {z=-980}
    physicsBodyDefinitions [{
      boundBone {boneName=dyn_lace_l_strand_r}
      boxExtents=<20 30 30>
      constraintSetup {angularLimitsMax=<15 15 15>, angularLimitsMin=<-15 -15 -15>, angularTarget=<1 0 0>}
      localJointOffset=<4 0 0>
     }]
   }
   n7 = Anim Dynamics - Bone: dyn_lace_l_loop_l!(ComponentPose=@n6, Alpha=1, GravityOverride:Vector="0,0,-980") -> (Pose) {
    angularDampingOverride=0.99
    angularSpringConstant=5000
    bAngularSpring=true
    bOverrideAngularDamping=true
    bUseGravityOverride=true
    gravityScale=0.35
    simulationSpace=RootRelative
    alphaBoolBlend {blendOption=Linear}
    boundBone {boneName=dyn_lace_l_loop_l}
    componentLinearAccScale=<1 1 1>
    componentLinearVelScale=<1 1 1>
    gravityOverride {z=-980}
    physicsBodyDefinitions [{
      boundBone {boneName=dyn_lace_l_loop_l}
      boxExtents=<20 30 30>
      constraintSetup {angularLimitsMax=<15 15 15>, angularLimitsMin=<-15 -15 -15>, angularTarget=<1 0 0>}
      localJointOffset=<4 0 0>
     }]
   }
   n8 = Anim Dynamics - Bone: dyn_lace_l_loop_r!(ComponentPose=@n7, Alpha=1, GravityOverride:Vector="0,0,-980") -> (Pose) {
    angularDampingOverride=0.99
    angularSpringConstant=5000
    bAngularSpring=true
    bOverrideAngularDamping=true
    bUseGravityOverride=true
    gravityScale=0.35
    simulationSpace=RootRelative
    alphaBoolBlend {blendOption=Linear}
    boundBone {boneName=dyn_lace_l_loop_r}
    componentLinearAccScale=<1 1 1>
    componentLinearVelScale=<1 1 1>
    gravityOverride {z=-980}
    physicsBodyDefinitions [{
      boundBone {boneName=dyn_lace_l_loop_r}
      boxExtents=<20 30 30>
      constraintSetup {angularLimitsMax=<15 15 15>, angularLimitsMin=<-15 -15 -15>, angularTarget=<1 0 0>}
      localJointOffset=<4 0 0>
     }]
   }
   n9 = Anim Dynamics - Bone: dyn_upper_lapel_l!(ComponentPose=@n8, Alpha=1, GravityOverride:Vector="0,0,-980") -> (Pose) {
    angularDampingOverride=1
    angularSpringConstant=5000
    bAngularSpring=true
    bOverrideAngularDamping=true
    bUseGravityOverride=true
    gravityScale=0.35
    simulationSpace=RootRelative
    alphaBoolBlend {blendOption=Linear}
    boundBone {boneName=dyn_upper_lapel_l}
    componentLinearAccScale=<1 1 1>
    componentLinearVelScale=<1 1 1>
    gravityOverride {z=-980}
    physicsBodyDefinitions [{
      boundBone {boneName=dyn_upper_lapel_l}
      boxExtents=<10 20 20>
      constraintSetup {angularLimitsMax=<10 10 10>, angularLimitsMin=<-10 -10 -10>, angularTarget=<1 0 0>}
      localJointOffset=<4 0 0>
     }]
   }
   n10 = Anim Dynamics - Bone: dyn_upper_lapel_r!(ComponentPose=@n9, Alpha=1, GravityOverride:Vector="0,0,-980") -> (Pose) {
    angularDampingOverride=1
    angularSpringConstant=5000
    bAngularSpring=true
    bOverrideAngularDamping=true
    bUseGravityOverride=true
    gravityScale=0.35
    simulationSpace=RootRelative
    alphaBoolBlend {blendOption=Linear}
    boundBone {boneName=dyn_upper_lapel_r}
    componentLinearAccScale=<1 1 1>
    componentLinearVelScale=<1 1 1>
    gravityOverride {z=-980}
    physicsBodyDefinitions [{
      boundBone {boneName=dyn_upper_lapel_r}
      boxExtents=<10 20 20>
      constraintSetup {angularLimitsMax=<10 10 10>, angularLimitsMin=<-10 -10 -10>, angularTarget=<1 0 0>}
      localJointOffset=<4 0 0>
     }]
   }
   n11 = Anim Dynamics - Bone: dyn_auxcord_end!(ComponentPose=@n10, Alpha=1, GravityOverride:Vector="0,0,-980") -> (Pose) {
    angularDampingOverride=1
    angularSpringConstant=5000
    bAngularSpring=true
    bOverrideAngularDamping=true
    bUseGravityOverride=true
    gravityScale=0.35
    simulationSpace=RootRelative
    alphaBoolBlend {blendOption=Linear}
    boundBone {boneName=dyn_auxcord_end}
    componentLinearAccScale=<1 1 1>
    componentLinearVelScale=<1 1 1>
    gravityOverride {z=-980}
    physicsBodyDefinitions [{
      boundBone {boneName=dyn_auxcord_end}
      boxExtents=<10 20 20>
      constraintSetup {angularLimitsMax=<10 10 10>, angularLimitsMin=<-10 -10 -10>, angularTarget=<1 0 0>}
      localJointOffset=<4 0 0>
     }]
   }
   Output Pose(Result=Component To Local(ComponentPose=@n11)) {layerGroup=None}
   n12 = Control Rig!(
    Source=Copy Pose From Mesh() {bUseAttachedParent=true}
    Alpha
    BeltFix:double=Get BeltFixAlpha() -> (BeltFixAlpha:double)
    ShoulderFix:double=Get ShoulderFixAlpha() -> (ShoulderFixAlpha:double)
    ArmpitFix:double=Get ArmpitFixAlpha() -> (ArmpitFixAlpha:double)
    Parachuting:bool=Get Parachuting() -> (Parachuting:bool)
    PlayingEmote:bool=Get isEmote() -> (isEmote:bool)
    PlayingEmote_cust:bool=Get isCustomEmotePlaying() -> (isCustomEmotePlaying:bool)
   ) -> (Pose) {
    alphaBoolBlend {blendOption=Linear}
    alphaScaleBiasClamp {bInterpResult=true}
    controlRigAssetReference {blueprintRigClass=&/F_MED_StrategicSpur_Mic_CtrlRig.F_MED_StrategicSpur_Mic_CtrlRig_C}
   }
   n13 = "Transform (Modify) Bone - Bone: base_auxcord_mic"!(ComponentPose=Local To Component(LocalPose=@n12), Alpha=1, Translation="1,0,0", Rotation="15,-25,-12") -> (Pose) {
    rotationMode=BMM_Additive
    translationMode=BMM_Additive
    alphaBoolBlend {blendInTime=5, blendOption=Linear, blendOutTime=5}
    alphaScaleBiasClamp {bInterpResult=true, interpSpeedIncreasing=7}
    boneToModify {boneName=base_auxcord_mic}
    rotation=<rot 15 -25 -12>
    translation {x=1}
   }
   n14 = RigidBody!(ComponentPose=@n13, Alpha, OverrideWorldGravity:Vector="0,0,0") -> (Pose) {
    bClampLinearTranslationLimitToRefPose=true
    bOverrideWorldGravity=true
    overridePhysicsAsset=&/F_MED_StrategicSpur_Mic_Lapels_Physics
    simulationSpace=BaseBoneSpace
    alphaBoolBlend {blendOption=Linear}
    baseBoneRef {boneName=spine_02}
    componentAppliedLinearAccClamp=<3000 3000 3000>
    componentLinearAccScale=<0.3 0.3 0.3>
    componentLinearVelScale=<0.3 0.3 0.3>
    overrideWorldGravity {z=-980}
    simSpaceSettings {maxAngularAcceleration=100, maxAngularVelocity=5, maxLinearAcceleration=300, maxLinearVelocity=50, velocityScaleZ=0}
   }
   n15 = RigidBody(ComponentPose=@n14, OverrideWorldGravity="0,0,0") {
    bClampLinearTranslationLimitToRefPose=true
    bOverrideWorldGravity=true
    overridePhysicsAsset=&/F_MED_StrategicSpur_Mic_Physics
    simulationSpace=BaseBoneSpace
    alphaBoolBlend {blendOption=Linear}
    baseBoneRef {boneName=pelvis}
    componentAppliedLinearAccClamp=<3000 3000 3000>
    componentLinearAccScale=<0.3 0.3 0.37>
    componentLinearVelScale=<0.3 0.3 0.37>
    overrideWorldGravity {z=-980}
    simSpaceSettings {
     maxAngularAcceleration=100
     maxAngularVelocity=5
     maxLinearAcceleration=300
     maxLinearVelocity=50
     velocityScaleZ=0
     worldAlpha=0.65
     externalLinearDragV=<0.2 0.2>
    }
   }
   Save cached pose 'base'(Pose=Component To Local(ComponentPose=@n15)) {cacheName=base}
   n16 = "Transform (Modify) Bone - Bone: base_sholderpad_f_l"!(ComponentPose=Local To Component(LocalPose=Use cached pose 'base'()), bAlphaBoolEnabled:bool=True, Translation="0,2,-1") -> (Pose) {
    alphaInputType=Bool
    rotationMode=BMM_Additive
    translationMode=BMM_Additive
    alphaBoolBlend {blendInTime=0.2, blendOption=Linear, blendOutTime=0.2}
    alphaScaleBiasClamp {bInterpResult=true, interpSpeedDecreasing=5}
    boneToModify {boneName=base_sholderpad_f_l}
    translation {y=2, z=-1}
   }
   n17 = "Transform (Modify) Bone - Bone: base_sholderpad_f_r"!(ComponentPose=@n16, bAlphaBoolEnabled:bool=True, Translation="0,0,-3") -> (Pose) {
    alphaInputType=Bool
    rotationMode=BMM_Additive
    translationMode=BMM_Additive
    alphaBoolBlend {blendInTime=0.2, blendOption=Linear, blendOutTime=0.2}
    alphaScaleBiasClamp {bInterpResult=true, interpSpeedDecreasing=5}
    boneToModify {boneName=base_sholderpad_f_r}
    translation {z=-3}
   }
   n18 = "Transform (Modify) Bone - Bone: base_sholderpad_r"!(ComponentPose=@n17, bAlphaBoolEnabled:bool=True, Rotation="5,0,0") -> (Pose) {
    alphaInputType=Bool
    rotationMode=BMM_Additive
    rotationSpace=BCS_BoneSpace
    translationMode=BMM_Additive
    translationSpace=BCS_BoneSpace
    alphaBoolBlend {blendInTime=0.2, blendOption=Linear, blendOutTime=0.2}
    boneToModify {boneName=base_sholderpad_r}
    rotation {pitch=5}
   }
   n19 = "Transform (Modify) Bone - Bone: base_sholderpad_b_l"!(ComponentPose=@n18, bAlphaBoolEnabled:bool=True, Translation="0.668,-0.252,1", Rotation="0.000,0.000,10.000") -> (Pose) {
    alphaInputType=Bool
    rotationMode=BMM_Additive
    translationMode=BMM_Additive
    alphaBoolBlend {blendInTime=0.2, blendOption=Linear, blendOutTime=0.2}
    alphaScaleBiasClamp {bInterpResult=true, interpSpeedDecreasing=5}
    boneToModify {boneName=base_sholderpad_b_l}
    rotation {roll=10}
    translation=<0.668 -0.252 1>
   }
   n20 = "Transform (Modify) Bone - Bone: dyn_auxcord_mic_2"!(ComponentPose=@n19, bAlphaBoolEnabled:bool=True, Translation="0.928,-0.000,0.000", Rotation="19.683,-3.616,-40.628") -> (Pose) {
    alphaInputType=Bool
    rotationMode=BMM_Additive
    translationMode=BMM_Additive
    alphaBoolBlend {blendInTime=0.2, blendOption=Linear, blendOutTime=0.2}
    alphaScaleBiasClamp {bInterpResult=true, interpSpeedDecreasing=5}
    boneToModify {boneName=dyn_auxcord_mic_2}
    rotation=<rot 19.683 -3.616 -40.628>
    translation {x=0.928}
   }
   n21 = "Transform (Modify) Bone - Bone: base_auxcord_mic"!(ComponentPose=@n20, bAlphaBoolEnabled:bool=True, Translation="1,0,0", Rotation="6.230,-7.980,13.599") -> (Pose) {
    alphaInputType=Bool
    rotationMode=BMM_Additive
    translationMode=BMM_Additive
    alphaBoolBlend {blendInTime=0.25, blendOption=Linear, blendOutTime=0.25}
    boneToModify {boneName=base_auxcord_mic}
    rotation=<rot 6.23 -7.98 13.599>
    translation {x=1}
   }
   n22 = "Transform (Modify) Bone - Bone: base_auxcord_mic"(ComponentPose=@n21, bAlphaBoolEnabled=True, Translation="1.5,0,0", Rotation="0,30,10") {
    alphaInputType=Bool
    rotationMode=BMM_Additive
    translationMode=BMM_Additive
    alphaBoolBlend {blendInTime=0.2, blendOption=Linear, blendOutTime=0.2}
    alphaScaleBiasClamp {bInterpResult=true, interpSpeedDecreasing=5}
    boneToModify {boneName=base_auxcord_mic}
    rotation {roll=10, yaw=30}
    translation {x=1.5}
   }
   n23 = "Transform (Modify) Bone - Bone: base_auxcord_mic"(ComponentPose=@n22, bAlphaBoolEnabled=True, Translation="2.5,2.5,0", Rotation="0,0,0") {
    alphaInputType=Bool
    rotationMode=BMM_Additive
    rotationSpace=BCS_BoneSpace
    translationMode=BMM_Additive
    alphaBoolBlend {blendInTime=0.2, blendOption=Linear, blendOutTime=0.2}
    alphaScaleBiasClamp {bInterpResult=true, interpSpeedIncreasing=7}
    boneToModify {boneName=base_auxcord_mic}
    translation=<2.5 2.5>
   }
   n24 = "Transform (Modify) Bone - Bone: dyn_auxcord_back_loop"!(ComponentPose=@n23, bAlphaBoolEnabled:bool=True, Rotation="-10,0,0") -> (Pose) {
    alphaInputType=Bool
    rotationMode=BMM_Additive
    rotationSpace=BCS_BoneSpace
    translationMode=BMM_Additive
    alphaBoolBlend {blendInTime=0.2, blendOption=Linear, blendOutTime=0.2}
    alphaScaleBiasClamp {bInterpResult=true, interpSpeedIncreasing=7}
    boneToModify {boneName=dyn_auxcord_back_loop}
    rotation {pitch=-10}
   }
   n25 = "Transform (Modify) Bone - Bone: base_auxcord_front_loop"!(ComponentPose=@n24, bAlphaBoolEnabled:bool=True, Rotation="-10,0,0") -> (Pose) {
    alphaInputType=Bool
    rotationMode=BMM_Additive
    rotationSpace=BCS_BoneSpace
    translationMode=BMM_Additive
    alphaBoolBlend {blendInTime=0.2, blendOption=Linear, blendOutTime=0.2}
    alphaScaleBiasClamp {bInterpResult=true, interpSpeedIncreasing=7}
    boneToModify {boneName=base_auxcord_front_loop}
    rotation {pitch=-10}
   }
   n26 = "Transform (Modify) Bone - Bone: base_auxcord_mic"(ComponentPose=@n25, bAlphaBoolEnabled=True, Translation="2.729,-1.122,-1.821", Rotation="-10.290,-14.737,-18.117") {
    alphaInputType=Bool
    rotationMode=BMM_Additive
    translationMode=BMM_Additive
    alphaBoolBlend {blendInTime=0.2, blendOption=Linear, blendOutTime=0.2}
    alphaScaleBiasClamp {bInterpResult=true, interpSpeedDecreasing=5}
    boneToModify {boneName=base_auxcord_mic}
    rotation=<rot -10.29 -14.737 -18.117>
    translation=<2.7285 -1.122 -1.821>
   }
   n27 = "Transform (Modify) Bone - Bone: base_shirtmover_r"(
    ComponentPose=@n26
    Alpha=Get UpperArmRot_r() -> (UpperArmRot_r:double)
    Translation="-0.884,-6.932,2.057"
    Rotation="-9.847,-1.754,10.151"
    Scale="1,1,1"
   ) {
    rotationMode=BMM_Additive
    rotationSpace=BCS_BoneSpace
    translationMode=BMM_Additive
    translationSpace=BCS_BoneSpace
    alphaBoolBlend {blendOption=Linear}
    alphaScaleBiasClamp {bInterpResult=true, interpSpeedDecreasing=5, interpSpeedIncreasing=5}
    boneToModify {boneName=base_shirtmover_r}
    rotation=<rot -9.8472 -1.754 10.151>
    translation=<-0.8837 -6.932 2.057>
   }
   n28 = Get spineRot() -> (spineRot:double)
   n29 = "Transform (Modify) Bone - Bone: base_lower_lapel_r"(ComponentPose=@n27, Alpha=@n28, Translation="0,1.5,0", Rotation="0,20,-15", Scale="1,1,1") {
    rotationMode=BMM_Additive
    rotationSpace=BCS_BoneSpace
    translationMode=BMM_Additive
    translationSpace=BCS_BoneSpace
    alphaBoolBlend {blendOption=Linear}
    alphaScaleBiasClamp {bInterpResult=true, interpSpeedDecreasing=8, interpSpeedIncreasing=8}
    boneToModify {boneName=base_lower_lapel_r}
    rotation {roll=-15, yaw=20}
    translation {y=1.5}
   }
   n30 = "Transform (Modify) Bone - Bone: base_upper_lapel_r"(ComponentPose=@n29, Alpha=@n28, Translation="0,0.5,0", Rotation="0,0,0", Scale="1,1,1") {
    rotationMode=BMM_Additive
    rotationSpace=BCS_BoneSpace
    translationMode=BMM_Additive
    translationSpace=BCS_BoneSpace
    alphaBoolBlend {blendOption=Linear}
    alphaScaleBiasClamp {bInterpResult=true, interpSpeedDecreasing=8, interpSpeedIncreasing=8}
    boneToModify {boneName=base_upper_lapel_r}
    translation {y=0.5}
   }
   Save cached pose 'Fixes'(Pose=Component To Local(ComponentPose=@n30)) {cacheName=Fixes}
  ]}
 EventGraph {nodes [
   Event BlueprintUpdateAnimation()
   > n1 = Sequence!(execute) -> (then_0, then_1, then_2:Exec, then_3:Exec)
   > n2 = Gravity Override State Function DEPRECATED(CharacterPartAnimInstance=Self-Reference(), GameTimeInSeconds=GetGameTimeInSeconds(), GravityOverrideParams=Get Gravity Override Params())
   > Set GravityOverride(GravityOverride=@n2)
   n3 = NOT Boolean(A=Get bIsMoving2D())
   n4 = Get bIsCrouching()
   > Set IsCrouchingStatic(IsCrouchingStatic:double="To Float (Boolean)"(InBool=AND Boolean(A=@n4, B=@n3))) -> (Output_Get:double)
   n5 = Get bIsPlayingEmote()
   > Set IsFEnoEmote(IsFEnoEmote:bool=AND Boolean(A=Get bIsInFrontEnd(), B=NOT Boolean(A=@n5))) -> (Output_Get:bool)
   n6 = GetOwnerSkeletalMeshComponent()
   n7 = BreakRotator(InRot=BreakTransform(InTransform=GetDeltaTransformFromRefPose(self=@n6, BoneName=clavicle_r, BaseName=spine_05)).Rotation)
   n8 = Branch(execute=@n1.then_1, Condition=AND Boolean(A="float >= float"(A=@n7.Yaw, B=3), B=Get bIsTargeting()))
   > Set UpperArmRot_r(UpperArmRot_r:double=1) -> (Output_Get:double)
   Set UpperArmRot_r(execute=@n8.else, UpperArmRot_r=0)
   n9 = GetOwnerSkeletalMeshComponent()
   n10 = BreakRotator(InRot=BreakTransform(InTransform=GetDeltaTransformFromRefPose(self=@n9, BoneName=clavicle_r, BaseName=spine_05)).Rotation)
   n11 = "float >= float"(A=@n10.Yaw, B=5)
   n12 = Get bIsPlayingEmote()
   n13 = "float >= float"(A=@n10.Yaw, B=19)
   n14 = Get bIsTargeting()
   n15 = Branch(execute=@n1.then_2, Condition=OR Boolean(A=AND Boolean(A=@n13, B=@n14), B=AND Boolean(A=@n11, B=@n12)))
   > Set SpineRot(SpineRot:double=1) -> (Output_Get:double)
   Set SpineRot(execute=@n15.else, SpineRot=0)
   n16 = GetAnimInstance(self=GetOwnerSkeletalMeshComponent())
   Is Valid(exec=@n1.then_3, InputObject=@n16)
   > Set isCustomEmotePlaying(isCustomEmotePlaying:bool=Montage_IsPlaying(self=@n16, Montage=/BRCosmetics/Animation/Game/MainPlayer/Emotes/Emote_Celebration_Female_M)) -> (Output_Get:bool)
   Event OnParachuteOpened()
   > Set ArmpitFixAlpha(ArmpitFixAlpha:double=0) -> (Output_Get:double)
   > Set Parachuting(Parachuting:bool=1) -> (Output_Get:bool)
   Event OnBeginPlayingEmote() {is=title K2Node_Event}
   > Set isEmote(isEmote:bool=1) -> (Output_Get:bool)
   Event OnEndPlayingEmote() {is=title K2Node_Event}
   > Set isEmote(isEmote=false)
   Event OnParachuteClosed()
   > Set ArmpitFixAlpha(ArmpitFixAlpha=1)
   > Set Parachuting(Parachuting=false)
   n17 = Event OnEndJumping()
   Set isJumping(execute=[@n17.then, @n19.then], isJumping:double=0) -> (Output_Get:double)
   n18 = Event OnBeginFalling()
   Set isJumping(execute=[@n18.then, @n20.then], isJumping=1)
   n19 = Event OnEndFalling()
   n20 = Event OnBeginJumping()
   n21 = Sequence!(execute) -> (then_0, then_1)
   n22 = Get bIsCrouching()
   > n23 = Branch(Condition=AND Boolean(A=Get bIsMoving2D(), B=NOT Boolean(A=@n22)))
   > Set BeltFixAlpha(BeltFixAlpha:double=0.75) -> (Output_Get:double)
   Set BeltFixAlpha(execute=@n23.else, BeltFixAlpha=1)
   n24 = Branch(
    execute=@n21.then_1
    Condition=OR Boolean!(A=Get bIsSkyDiving(), B=Get bIsParachuteOpened(), C:bool=Get bIsSkydiveFloating(), D:bool=Get bIsSkydiveDiveMode()) -> (ReturnValue)
   )
   > Set ShoulderFixAlpha(ShoulderFixAlpha:double=0) -> (Output_Get:double)
   n25 = Branch(execute=@n24.else, Condition=Get bIsTacticalSprinting())
   > Set ShoulderFixAlpha(ShoulderFixAlpha=2)
   n26 = Branch(execute=@n25.else, Condition=OR Boolean!(A=Get bIsJumping(), B=Get bIsFalling()) -> (ReturnValue))
   > Set ShoulderFixAlpha(ShoulderFixAlpha=0.65)
   Set ShoulderFixAlpha(execute=@n26.else, ShoulderFixAlpha=1)
  ]}
]
variables [
 ArmpitFixAlpha {type=number}
 BeltFixAlpha {type=number}
 Gravity Override Params {type=GravityOverrideParamsStruct}
 GravityOverride {type=Vector}
 IdlePose {type=number}
 IsCrouchingStatic {type=number}
 isCustomEmotePlaying {type=boolean}
 isEmote {type=boolean}
 IsFEnoEmote {type=boolean}
 isJumping {type=number}
 ParachuteOpenEmote {type=boolean}
 Parachuting {type=boolean}
 PlayerPawn {type=FortPlayerPawnAthena}
 ShoulderFixAlpha {type=number}
 skelMesh {type=boolean}
 SpineRot {type=number}
 UpperArmRot_r {type=number}
]
