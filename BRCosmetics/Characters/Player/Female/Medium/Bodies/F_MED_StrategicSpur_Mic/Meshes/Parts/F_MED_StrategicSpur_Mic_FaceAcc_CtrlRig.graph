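# Control Rig for the StrategicSpur Mic female face accessory part, applying a head-pose-driven corrective to the last hawk bone.
# A spherical pose reader, with debug drawing left on, measures the head bone's orientation relative to spine_04.
# A Sequence runs the reader first, then its output weights a Modify Transforms that rotates and offsets dyn_hawk_6.
# The other hawk and earring bones are in the hierarchy but untouched by the graph.
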ControlRigBlueprint /BRCosmetics/Characters/Player/Female/Medium/Bodies/F_MED_StrategicSpur_Mic/Meshes/Parts/F_MED_StrategicSpur_Mic_FaceAcc_CtrlRig
graphs [
 RigVMFunctionLibrary {links [], nodes []}
 RigVMModel {nodes [
   RigUnit_BeginExecution = &RigUnit_BeginExecution()
   RigVMFunction_Sequence = &RigVMFunction_Sequence(ExecuteContext=@RigUnit_BeginExecution.ExecutePin)
   n1 = &RigUnit_SphericalPoseReader(
    ActiveRegionScaleFactors=(PositiveWidth=3,NegativeWidth=2,PositiveHeight=6,NegativeHeight=6)
    ActiveRegionSize=0.1
    Debug=(bDrawDebug=True,DebugScale=25,DebugSegments=20,DebugThickness=0.25)
    DriverAxis=(X=1,Y=0,Z=0)
    DriverItem=(Type=Bone,Name="head")
    FalloffRegionScaleFactors=(PositiveWidth=-3,NegativeWidth=3,PositiveHeight=0,NegativeHeight=0)
    FalloffSize=0.25
    OptionalParentItem=(Type=Bone,Name="spine_04")
    RotationOffset=(X=90,Y=0,Z=-180)
    ExecutePin=@RigVMFunction_Sequence.A
   )
   &RigUnit_ModifyTransforms(
    ItemToModify=((Item=(Type=Bone,Name="dyn_hawk_6"),Transform=(Rotation=(X=0,Y=0,Z=0,W=1),Translation=(X=5,Y=0,Z=0),Scale3D=(X=1,Y=1,Z=1))))
    ItemToModify.0.Transform.Rotation=&RigVMFunction_MathQuaternionFromEuler(Euler=(X=0,Y=0,Z=90)).Result
    ExecutePin=@RigVMFunction_Sequence.B
    Weight=@n1.OutputParam
   )
  ]}
]
hierarchy [
 root {loc=<0 0 0>, rot=<0 0 0 1>}
 pelvis {parent=root, loc=<0 -0.5301 98.0571>, rot=<0.0127 0.707 -0.0127 -0.707>}
 spine_01 {parent=pelvis, loc=<6.1208 -0.9328 0>, rot=<0 0 0.034 -0.9994>}
 spine_02 {parent=spine_01, loc=<6.5996 -0.2095 0>, rot=<0 0 0.0238 -0.9997>}
 spine_03 {parent=spine_02, loc=<6.4914 -0.1237 0>, rot=<0 0 0.0149 -0.9999>}
 spine_04 {parent=spine_03, loc=<6.9598 0.0794 0>, rot=<0 0 -0.0132 -0.9999>}
 spine_05 {parent=spine_04, loc=<6.9987 0 0>, rot=<0 0 -0.0015 -1>}
 attach_cape {parent=spine_05, loc=<9.3216 -8.4501 0>, rot=<0.0616 -0.7044 -0.0616 -0.7044>}
 neck_01 {parent=spine_05, loc=<11.3138 0.1529 0>, rot=<0 0 -0.2003 -0.9797>}
 neck_02 {parent=neck_01, loc=<4.5627 0 0>, rot=<0 0 0.0504 -0.9987>}
 head {parent=neck_02, loc=<5.3914 0 0>, rot=<0 0 0.115 -0.9934>}
 base_earring_r1 {parent=head, loc=<0.5733 0.2572 6.102>, rot=<0 0 1 0>}
 dyn_earring_r {parent=base_earring_r1, loc=<0 0 0>, rot=<0 0 0 1>}
 dyn_earring_2_r {parent=dyn_earring_r, loc=<1.099 0.0291 0>, rot=<0 0 0 1>}
 base_earring_l {parent=head, loc=<0.5337 0.7643 -5.8781>, rot=<0 0 1 0>}
 dyn_earring_l {parent=base_earring_l, loc=<0 0 0>, rot=<0 0 0 1>}
 base_hawk_1 {parent=head, loc=<12.102 7.6733 0>, rot=<0 0 0 1>}
 dyn_hawk_1 {parent=base_hawk_1, loc=<0 0 0>, rot=<0 0 0 1>}
 base_hawk_2 {parent=head, loc=<14.0157 2.2321 0>, rot=<0 0 0 1>}
 dyn_hawk_2 {parent=base_hawk_2, loc=<0 0 0>, rot=<0 0 0 1>}
 base_hawk_3 {parent=head, loc=<13.2768 -2.7691 0>, rot=<0 0 0.1736 -0.9848>}
 dyn_hawk_3 {parent=base_hawk_3, loc=<0 0 0>, rot=<0 0 0 1>}
 base_hawk_4 {parent=head, loc=<9.423 -6.7982 0>, rot=<0 0 0.1736 -0.9848>}
 dyn_hawk_4 {parent=base_hawk_4, loc=<0 0 0>, rot=<0 0 0 1>}
 base_hawk_5 {parent=head, loc=<4.3582 -6.7022 0>, rot=<0 0 0.7071 -0.7071>}
 dyn_hawk_5 {parent=base_hawk_5, loc=<0 0 0>, rot=<0 0 0 1>}
 base_hawk_6 {parent=head, loc=<0.2213 -5.5581 0>, rot=<0 0 0.9063 -0.4226>}
 dyn_hawk_6 {parent=base_hawk_6, loc=<0 0 0>, rot=<0 0 0 1>}
]
