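# Physics asset for the face accessory part of the StrategicSpur_Mic female medium character cosmetic, simulating dangling earrings against the head.
# A head capsule anchors small capsules on the base earring bones, joined by free swing and twist constraints.
# Dynamic earring bones hang from the bases with tighter angular limits.
# The sides are asymmetric: the right chain has an extra, larger end capsule with wider limits, while the left has one dynamic segment.
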PhysicsAsset /BRCosmetics/Characters/Player/Female/Medium/Bodies/F_MED_StrategicSpur_Mic/Meshes/Parts/F_MED_StrategicSpur_Mic_FaceAcc_Physics
bodies {
 base_earring_l [{shapeType=Capsule, length=0.5, radius=0.5, shapeName=base_earring_l_capsule, center=<0.3 0 0>, rotation=<rot 90 0 0>}]
 base_earring_r1 [{shapeType=Capsule, length=0.5, radius=0.5, shapeName=base_earring_r1_capsule, rotation=<rot 90 0 0>}]
 dyn_earring_2_r [{shapeType=Capsule, length=1, radius=1.5, shapeName=dyn_earring_2_r_capsule, center=<4.9746 0 0>, rotation=<rot 90 0 0>}]
 dyn_earring_l [{shapeType=Capsule, length=0.5, radius=0.5, shapeName=dyn_earring_l_capsule, rotation=<rot 90 0 0>}]
 dyn_earring_r [{shapeType=Capsule, length=0.5, radius=0.5, shapeName=dyn_earring_r_capsule, center=<0.3 0 0>, rotation=<rot 90 0 0>}]
 head [{shapeType=Capsule, length=10.5, radius=4, shapeName=head_capsule, center=<-4 0 0>, rotation=<rot 0 0 90>}]
}
constraints [
 {
  bone1Name=base_earring_l
  bone2Name=head
  swing1LimitDegrees=45
  swing1Motion=Free
  swing2LimitDegrees=45
  swing2Motion=Free
  twistLimitDegrees=45
  twistMotion=Free
 }
 {bone1Name=dyn_earring_l, bone2Name=base_earring_l, swing1LimitDegrees=10, swing2LimitDegrees=10, twistLimitDegrees=0}
 {
  bone1Name=base_earring_r1
  bone2Name=head
  swing1LimitDegrees=45
  swing1Motion=Free
  swing2LimitDegrees=45
  swing2Motion=Free
  twistLimitDegrees=45
  twistMotion=Free
 }
 {bone1Name=dyn_earring_r, bone2Name=base_earring_r1, swing1LimitDegrees=20, swing2LimitDegrees=20, twistLimitDegrees=0}
 {bone1Name=dyn_earring_2_r, bone2Name=dyn_earring_r, swing1LimitDegrees=50, swing2LimitDegrees=60, twistLimitDegrees=10}
]
